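# Physics asset for the medium male Noble cloak (FrontENd variant), giving its dynamic skirt bones capsule bodies for physical motion.
# Seven chains (B, D, E left and right, F center) hang from master bodies joined to spine_03 by free-rotation constraints.
# Chain links have narrow swing limits and mostly zero twist limits; E and F skip segment 02.
# Spine, pelvis, leg and foot capsules form a freely jointed body proxy, with extra thigh shapes.
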PhysicsAsset /Game/Characters/Player/Male/Medium/Bodies/M_MED_Noble/Meshes/M_MED_Noble_Cloak_FrontENd_Physics
bodies {
 calf_l [{shapeType=Capsule, length=50.1993, radius=8.2322, shapeName=calf_l_capsule, center=<-19.8973 0 0>, rotation=<rot 90 0 0>}]
 calf_r [{shapeType=Capsule, length=50.1993, radius=8.2322, shapeName=calf_l_capsule, center=<19.8973 0 0>, rotation=<rot 90 0 0>}]
 dyn_SkirtB01_MSTR_l [{shapeType=Capsule, length=1, radius=1, shapeName=dyn_SkirtB01_MSTR_l_capsule}]
 dyn_SkirtB01_MSTR_r [{shapeType=Capsule, length=1, radius=1, shapeName=dyn_SkirtB01_MSTR_r_capsule}]
 dyn_SkirtB01_l [{shapeType=Capsule, length=11.8568, radius=2.6253, shapeName=dyn_SkirtB01_l_capsule, center=<3.5821 0 0>, rotation=<rot 90 0 0>}]
 dyn_SkirtB01_r [{shapeType=Capsule, length=11.8568, radius=2.6253, shapeName=dyn_SkirtB01_r_capsule, center=<3.5821 0 0>}]
 dyn_SkirtB02_l [{shapeType=Capsule, length=13.7442, radius=2.2159, shapeName=dyn_SkirtB02_l_capsule, center=<4.3161 0 0>, rotation=<rot 90 0 0>}]
 dyn_SkirtB02_r [{shapeType=Capsule, length=12.2442, radius=3.7159, shapeName=dyn_SkirtB02_r_capsule, center=<4.3161 0 0>}]
 dyn_SkirtB03_l [{shapeType=Capsule, length=15.2955, radius=3.0757, shapeName=dyn_SkirtB03_l_capsule, center=<6.8844 0 0>, rotation=<rot 90 0 0>}]
 dyn_SkirtB03_r [{shapeType=Capsule, length=13.5455, radius=4.8257, shapeName=dyn_SkirtB03_r_capsule, center=<6.8844 0 0>}]
 dyn_SkirtB04_l [{shapeType=Capsule, length=11.0532, radius=3.5393, shapeName=dyn_SkirtB04_l_capsule, center=<6.8844 0 0>, rotation=<rot 90 0 0>}]
 dyn_SkirtB04_r [{shapeType=Capsule, length=13.827, radius=3.8324, shapeName=dyn_SkirtB04_r_capsule, center=<7.8376 0 0>}]
 dyn_SkirtB05_l [{shapeType=Capsule, length=11.3397, radius=3.5926, shapeName=dyn_SkirtB05_l_capsule, center=<6.8844 0 0>, rotation=<rot 90 0 0>}]
 dyn_SkirtB05_r [{shapeType=Capsule, length=12.4267, radius=3.9238, shapeName=dyn_SkirtB05_r_capsule, center=<6.8844 0 0>}]
 dyn_SkirtB06_l [{shapeType=Capsule, length=15.6729, radius=5.0078, shapeName=dyn_SkirtB06_l_capsule, center=<6.8844 0 0>, rotation=<rot 90 0 0>}]
 dyn_SkirtB06_r [{shapeType=Capsule, length=15.0052, radius=5.0078, shapeName=dyn_SkirtB06_r_capsule, center=<6.8844 0 0>, rotation=<rot 90 0 0>}]
 dyn_SkirtB07_l [{shapeType=Capsule, length=9.7403, radius=5.192, shapeName=dyn_SkirtB07_l_capsule, center=<6.8844 0 0>, rotation=<rot 90 0 0>}]
 dyn_SkirtB07_r [{shapeType=Capsule, length=8.4903, radius=6.442, shapeName=dyn_SkirtB07_r_capsule, center=<6.8844 0 0>, rotation=<rot 90 0 0>}]
 dyn_SkirtD01_MSTR_l [{shapeType=Capsule, length=1, radius=1, shapeName=dyn_SkirtD01_MSTR_l_capsule}]
 dyn_SkirtD01_MSTR_r [{shapeType=Capsule, length=1, radius=1, shapeName=dyn_SkirtD01_MSTR_r_capsule}]
 dyn_SkirtD01_l [{shapeType=Capsule, length=11.2263, radius=3.2558, shapeName=dyn_SkirtD01_l_capsule, center=<3.5821 0 0>, rotation=<rot 90 0 0>}]
 dyn_SkirtD01_r [{shapeType=Capsule, length=15.4322, radius=2.2317, shapeName=dyn_SkirtD01_r_capsule, center=<7.2001 0 0>, rotation=<rot 90 0 0>}]
 dyn_SkirtD02_l [{shapeType=Capsule, length=12.7805, radius=3.1796, shapeName=dyn_SkirtD02_l_capsule, center=<4.3161 0 0>, rotation=<rot 90 0 0>}]
 dyn_SkirtD02_r [{shapeType=Capsule, length=14.091, radius=1.869, shapeName=dyn_SkirtD02_r_capsule, center=<8.4064 0 0>, rotation=<rot 90 0 0>}]
 dyn_SkirtD03_l [{shapeType=Capsule, length=14.2955, radius=4.0757, shapeName=dyn_SkirtD03_l_capsule, center=<6.8844 0 0>, rotation=<rot 90 0 0>}]
 dyn_SkirtD03_r [{shapeType=Capsule, length=15.5455, radius=2.8257, shapeName=dyn_SkirtD03_r_capsule, center=<6.8844 0 0>, rotation=<rot 90 0 0>}]
 dyn_SkirtD04_l [{shapeType=Capsule, length=15.2727, radius=5.1581, shapeName=dyn_SkirtD04_l_capsule, center=<6.8844 0 0>}]
 dyn_SkirtD04_r [{shapeType=Capsule, length=14.2727, radius=4.4081, shapeName=dyn_SkirtD04_r_capsule, center=<6.8844 0 0>}]
 dyn_SkirtD05_l [{shapeType=Capsule, length=19.8193, radius=4.7696, shapeName=dyn_SkirtD05_l_capsule, center=<6.8844 0 0>}]
 dyn_SkirtD05_r [{shapeType=Capsule, length=12.5206, radius=4.8182, shapeName=dyn_SkirtD05_r_capsule, center=<6.8844 0 0>}]
 dyn_SkirtD06_l [{shapeType=Capsule, length=25.0191, radius=5.0634, shapeName=dyn_SkirtD06_l_capsule, center=<6.8844 0 0>}]
 dyn_SkirtD06_r [{shapeType=Capsule, length=13.4449, radius=5.7199, shapeName=dyn_SkirtD06_r_capsule, center=<6.8844 0 0>}]
 dyn_SkirtD07_l [{shapeType=Capsule, length=20.4131, radius=6.6952, shapeName=dyn_SkirtD07_l_capsule, center=<6.8844 0 0>}]
 dyn_SkirtD07_r [{shapeType=Capsule, length=9.6123, radius=6.4367, shapeName=dyn_SkirtD07_r_capsule, center=<6.8844 0 0>}]
 dyn_SkirtE01_MSTR_l [{shapeType=Capsule, length=1, radius=1, shapeName=dyn_SkirtE01_MSTR_l_capsule}]
 dyn_SkirtE01_MSTR_r [{shapeType=Capsule, length=1, radius=1, shapeName=dyn_SkirtE01_MSTR_r_capsule}]
 dyn_SkirtE01_l [{shapeType=Capsule, length=16.0183, radius=3.487, shapeName=dyn_SkirtE01_l_capsule, center=<8.8899 0 0>, rotation=<rot 90 0 0>}]
 dyn_SkirtE01_r [{shapeType=Capsule, length=15.2436, radius=3.2929, shapeName=dyn_SkirtE01_r_capsule, center=<8.8899 0 0>, rotation=<rot 90 0 0>}]
 dyn_SkirtE03_l [{shapeType=Capsule, length=13.0829, radius=5.2884, shapeName=dyn_SkirtE03_l_capsule, center=<6.8844 0 0>, rotation=<rot 90 0 0>}]
 dyn_SkirtE03_r [{shapeType=Capsule, length=15.5779, radius=4.0951, shapeName=dyn_SkirtE03_r_capsule, center=<6.8844 0 0>, rotation=<rot 90 0 0>}]
 dyn_SkirtE04_l [{shapeType=Capsule, length=14.6149, radius=5.8159, shapeName=dyn_SkirtE04_l_capsule, center=<6.8844 0 0>}]
 dyn_SkirtE04_r [{shapeType=Capsule, length=15.8649, radius=4.5659, shapeName=dyn_SkirtE04_r_capsule, center=<6.8844 0 0>}]
 dyn_SkirtE05_l [{shapeType=Capsule, length=19.2493, radius=5.3396, shapeName=dyn_SkirtE05_l_capsule, center=<6.8844 0 0>}]
 dyn_SkirtE05_r [{shapeType=Capsule, length=19.2493, radius=5.3396, shapeName=dyn_SkirtE05_r_capsule, center=<6.8844 0 0>}]
 dyn_SkirtE06_l [{shapeType=Capsule, length=20.3448, radius=6.0459, shapeName=dyn_SkirtE06_l_capsule, center=<6.8844 0 0>}]
 dyn_SkirtE06_r [{shapeType=Capsule, length=20.3448, radius=6.0459, shapeName=dyn_SkirtE06_r_capsule, center=<6.8844 0 0>}]
 dyn_SkirtE07_l [{shapeType=Capsule, length=19.2478, radius=7.1429, shapeName=dyn_SkirtE07_l_capsule, center=<6.8844 0 0>}]
 dyn_SkirtE07_r [{shapeType=Capsule, length=19.2478, radius=7.1429, shapeName=dyn_SkirtE07_r_capsule, center=<6.8844 0 0>}]
 dyn_SkirtF01_MSTR_c [{shapeType=Capsule, length=1, radius=1, shapeName=dyn_SkirtF01_MSTR_c_capsule}]
 dyn_SkirtF01_c [{shapeType=Capsule, length=13.9437, radius=3.7278, shapeName=dyn_SkirtF01_c_capsule, center=<8.8899 0 0>, rotation=<rot 90 0 0>}]
 dyn_SkirtF03_c [{shapeType=Capsule, length=11.5428, radius=6.8284, shapeName=dyn_SkirtF03_c_capsule, center=<6.8844 0 0>, rotation=<rot 90 0 0>}]
 dyn_SkirtF04_c [{shapeType=Capsule, length=13.6091, radius=6.8216, shapeName=dyn_SkirtF04_c_capsule, center=<6.8844 0 0>, rotation=<rot 0 0 90>}]
 dyn_SkirtF05_c [{shapeType=Capsule, length=25.9101, radius=6.7191, shapeName=dyn_SkirtF05_c_capsule, center=<6.8844 0 0>, rotation=<rot 0 0 90>}]
 dyn_SkirtF06_c [{shapeType=Capsule, length=30.0003, radius=6.4711, shapeName=dyn_SkirtF06_c_capsule, center=<6.8844 0 0>, rotation=<rot 0 0 90>}]
 dyn_SkirtF07_c [{shapeType=Capsule, length=31.3887, radius=7.1429, shapeName=dyn_SkirtF07_c_capsule, center=<6.8844 0 -3.1346>, rotation=<rot 0 0 90>}]
 foot_l [{
   shapeType=Capsule
   length=24.8595
   radius=7.4737
   shapeName=foot_l_capsule
   center=<-4.6293 7.3879 0.3625>
   rotation=<rot 0 -0.2319 -85.9858>
  }]
 foot_r [{shapeType=Capsule, length=24.8366, radius=7.3191, shapeName=foot_r_capsule, center=<4.5106 -8.7279 0.1667>, rotation=<rot 0 0.4931 -85.4534>}]
 pelvis [
  {shapeType=Capsule, length=10.6765, radius=10.4017, shapeName=pelvis_capsule, center=<0 -0.7759 0>}
  {shapeType=Capsule, length=12.0697, radius=10.8793, shapeName=pelvis_capsule, center=<-9.4262 0 0>}
 ]
 spine_01 [{shapeType=Capsule, length=13.4709, radius=8.8186, shapeName=spine_01_capsule, center=<0 -1.4149 0>}]
 spine_02 [{shapeType=Capsule, length=10.9147, radius=9.6144, shapeName=spine_02_capsule, center=<0 -1.7658 0>}]
 spine_03 [{shapeType=Capsule, length=0.505, radius=0.505, shapeName=spine_03_capsule}]
 thigh_l [
  {shapeType=Capsule, length=29.6606, radius=8.5069, shapeName=thigh_l_capsule, center=<-19.8973 0 0>, rotation=<rot 90 0 0>}
  {
   shapeType=Capsule
   length=28.0955
   radius=7.0863
   shapeName=thigh_l_capsule
   center=<-19.5226 8.7061 6.9287>
   rotation=<rot 86.1882 0 0.0001>
  }
  {shapeType=Capsule, length=23.3029, radius=6.6971, shapeName=thigh_l_capsule, center=<-19.8467 17.5044 7.7115>, rotation=<rot 90 0 0>}
 ]
 thigh_r [
  {shapeType=Capsule, length=29.6606, radius=8.5069, shapeName=thigh_l_capsule, center=<19.8973 0 0>, rotation=<rot 90 0 0>}
  {
   shapeType=Capsule
   length=28.0955
   radius=7.0863
   shapeName=thigh_l_capsule
   center=<19.5697 -8.5465 -6.2224>
   rotation=<rot 86.1882 0 0.0001>
  }
 ]
}
constraints [
 {bone1Name=dyn_SkirtB01_r, bone2Name=dyn_SkirtB01_MSTR_r, swing1LimitDegrees=10, swing2LimitDegrees=10, twistLimitDegrees=20}
 {bone1Name=dyn_SkirtB02_r, bone2Name=dyn_SkirtB01_r, swing1LimitDegrees=10, swing2LimitDegrees=10, twistLimitDegrees=20}
 {bone1Name=dyn_SkirtB03_r, bone2Name=dyn_SkirtB02_r, swing1LimitDegrees=20, swing2LimitDegrees=20, twistLimitDegrees=20}
 {bone1Name=dyn_SkirtB04_r, bone2Name=dyn_SkirtB03_r, swing1LimitDegrees=20, swing2LimitDegrees=20, twistLimitDegrees=20}
 {bone1Name=dyn_SkirtB05_r, bone2Name=dyn_SkirtB04_r, swing1LimitDegrees=20, swing2LimitDegrees=20, twistLimitDegrees=20}
 {bone1Name=dyn_SkirtB06_r, bone2Name=dyn_SkirtB05_r, swing1LimitDegrees=20, swing2LimitDegrees=20, twistLimitDegrees=20}
 {bone1Name=dyn_SkirtB07_r, bone2Name=dyn_SkirtB06_r, swing1LimitDegrees=20, swing2LimitDegrees=20, twistLimitDegrees=20}
 {bone1Name=dyn_SkirtE01_r, bone2Name=dyn_SkirtE01_MSTR_r, swing1LimitDegrees=10, swing2LimitDegrees=10, twistLimitDegrees=0}
 {bone1Name=dyn_SkirtE04_r, bone2Name=dyn_SkirtE03_r, swing1LimitDegrees=20, swing2LimitDegrees=20, twistLimitDegrees=0}
 {bone1Name=dyn_SkirtE05_r, bone2Name=dyn_SkirtE04_r, swing1LimitDegrees=20, swing2LimitDegrees=20, twistLimitDegrees=0}
 {bone1Name=dyn_SkirtE06_r, bone2Name=dyn_SkirtE05_r, swing1LimitDegrees=20, swing2LimitDegrees=20, twistLimitDegrees=0}
 {bone1Name=dyn_SkirtE07_r, bone2Name=dyn_SkirtE06_r, swing1LimitDegrees=20, swing2LimitDegrees=20, twistLimitDegrees=0}
 {bone1Name=dyn_SkirtD01_r, bone2Name=dyn_SkirtD01_MSTR_r, swing1LimitDegrees=10, swing2LimitDegrees=10, twistLimitDegrees=0}
 {bone1Name=dyn_SkirtD02_r, bone2Name=dyn_SkirtD01_r, swing1LimitDegrees=10, swing2LimitDegrees=10, twistLimitDegrees=0}
 {bone1Name=dyn_SkirtD03_r, bone2Name=dyn_SkirtD02_r, swing1LimitDegrees=20, swing2LimitDegrees=20, twistLimitDegrees=0}
 {bone1Name=dyn_SkirtD04_r, bone2Name=dyn_SkirtD03_r, swing1LimitDegrees=20, swing2LimitDegrees=20, twistLimitDegrees=0}
 {bone1Name=dyn_SkirtD05_r, bone2Name=dyn_SkirtD04_r, swing1LimitDegrees=20, swing2LimitDegrees=20, twistLimitDegrees=0}
 {bone1Name=dyn_SkirtD06_r, bone2Name=dyn_SkirtD05_r, swing1LimitDegrees=20, swing2LimitDegrees=20, twistLimitDegrees=0}
 {bone1Name=dyn_SkirtD07_r, bone2Name=dyn_SkirtD06_r, swing1LimitDegrees=20, swing2LimitDegrees=20, twistLimitDegrees=0}
 {bone1Name=dyn_SkirtF01_c, bone2Name=dyn_SkirtF01_MSTR_c, swing1LimitDegrees=10, swing2LimitDegrees=10, twistLimitDegrees=0}
 {bone1Name=dyn_SkirtF04_c, bone2Name=dyn_SkirtF03_c, swing1LimitDegrees=20, swing2LimitDegrees=20, twistLimitDegrees=0}
 {bone1Name=dyn_SkirtF05_c, bone2Name=dyn_SkirtF04_c, swing1LimitDegrees=20, swing2LimitDegrees=20, twistLimitDegrees=0}
 {bone1Name=dyn_SkirtF06_c, bone2Name=dyn_SkirtF05_c, swing1LimitDegrees=20, swing2LimitDegrees=20, twistLimitDegrees=0}
 {bone1Name=dyn_SkirtF07_c, bone2Name=dyn_SkirtF06_c, swing1LimitDegrees=20, swing2LimitDegrees=20, twistLimitDegrees=0}
 {bone1Name=dyn_SkirtE01_l, bone2Name=dyn_SkirtE01_MSTR_l, swing1LimitDegrees=10, swing2LimitDegrees=10, twistLimitDegrees=0}
 {bone1Name=dyn_SkirtE04_l, bone2Name=dyn_SkirtE03_l, swing1LimitDegrees=20, swing2LimitDegrees=20, twistLimitDegrees=0}
 {bone1Name=dyn_SkirtE05_l, bone2Name=dyn_SkirtE04_l, swing1LimitDegrees=20, swing2LimitDegrees=20, twistLimitDegrees=0}
 {bone1Name=dyn_SkirtE06_l, bone2Name=dyn_SkirtE05_l, swing1LimitDegrees=20, swing2LimitDegrees=20, twistLimitDegrees=0}
 {bone1Name=dyn_SkirtE07_l, bone2Name=dyn_SkirtE06_l, swing1LimitDegrees=20, swing2LimitDegrees=20, twistLimitDegrees=0}
 {bone1Name=dyn_SkirtD01_l, bone2Name=dyn_SkirtD01_MSTR_l, swing1LimitDegrees=10, swing2LimitDegrees=10, twistLimitDegrees=0}
 {bone1Name=dyn_SkirtD02_l, bone2Name=dyn_SkirtD01_l, swing1LimitDegrees=10, swing2LimitDegrees=10, twistLimitDegrees=0}
 {bone1Name=dyn_SkirtD03_l, bone2Name=dyn_SkirtD02_l, swing1LimitDegrees=20, swing2LimitDegrees=20, twistLimitDegrees=0}
 {bone1Name=dyn_SkirtD04_l, bone2Name=dyn_SkirtD03_l, swing1LimitDegrees=20, swing2LimitDegrees=20, twistLimitDegrees=0}
 {bone1Name=dyn_SkirtD05_l, bone2Name=dyn_SkirtD04_l, swing1LimitDegrees=20, swing2LimitDegrees=20, twistLimitDegrees=0}
 {bone1Name=dyn_SkirtD06_l, bone2Name=dyn_SkirtD05_l, swing1LimitDegrees=20, swing2LimitDegrees=20, twistLimitDegrees=0}
 {bone1Name=dyn_SkirtD07_l, bone2Name=dyn_SkirtD06_l, swing1LimitDegrees=20, swing2LimitDegrees=20, twistLimitDegrees=0}
 {bone1Name=dyn_SkirtB01_l, bone2Name=dyn_SkirtB01_MSTR_l, swing1LimitDegrees=20, swing2LimitDegrees=20, twistLimitDegrees=0}
 {bone1Name=dyn_SkirtB02_l, bone2Name=dyn_SkirtB01_l, swing1LimitDegrees=20, swing2LimitDegrees=20, twistLimitDegrees=0}
 {bone1Name=dyn_SkirtB03_l, bone2Name=dyn_SkirtB02_l, swing1LimitDegrees=20, swing2LimitDegrees=20, twistLimitDegrees=0}
 {bone1Name=dyn_SkirtB04_l, bone2Name=dyn_SkirtB03_l, swing1LimitDegrees=20, swing2LimitDegrees=20, twistLimitDegrees=0}
 {bone1Name=dyn_SkirtB05_l, bone2Name=dyn_SkirtB04_l, swing1LimitDegrees=20, swing2LimitDegrees=20, twistLimitDegrees=0}
 {bone1Name=dyn_SkirtB06_l, bone2Name=dyn_SkirtB05_l, swing1LimitDegrees=20, swing2LimitDegrees=20, twistLimitDegrees=0}
 {bone1Name=dyn_SkirtB07_l, bone2Name=dyn_SkirtB06_l, swing1LimitDegrees=20, swing2LimitDegrees=20, twistLimitDegrees=0}
 {
  bone1Name=dyn_SkirtB01_MSTR_r
  bone2Name=spine_03
  swing1LimitDegrees=45
  swing1Motion=Free
  swing2LimitDegrees=45
  swing2Motion=Free
  twistLimitDegrees=45
  twistMotion=Free
 }
 {
  bone1Name=dyn_SkirtE01_MSTR_r
  bone2Name=spine_03
  swing1LimitDegrees=45
  swing1Motion=Free
  swing2LimitDegrees=45
  swing2Motion=Free
  twistLimitDegrees=45
  twistMotion=Free
 }
 {
  bone1Name=dyn_SkirtD01_MSTR_r
  bone2Name=spine_03
  swing1LimitDegrees=45
  swing1Motion=Free
  swing2LimitDegrees=45
  swing2Motion=Free
  twistLimitDegrees=45
  twistMotion=Free
 }
 {
  bone1Name=dyn_SkirtF01_MSTR_c
  bone2Name=spine_03
  swing1LimitDegrees=45
  swing1Motion=Free
  swing2LimitDegrees=45
  swing2Motion=Free
  twistLimitDegrees=45
  twistMotion=Free
 }
 {
  bone1Name=dyn_SkirtE01_MSTR_l
  bone2Name=spine_03
  swing1LimitDegrees=45
  swing1Motion=Free
  swing2LimitDegrees=45
  swing2Motion=Free
  twistLimitDegrees=45
  twistMotion=Free
 }
 {
  bone1Name=dyn_SkirtD01_MSTR_l
  bone2Name=spine_03
  swing1LimitDegrees=45
  swing1Motion=Free
  swing2LimitDegrees=45
  swing2Motion=Free
  twistLimitDegrees=45
  twistMotion=Free
 }
 {
  bone1Name=dyn_SkirtB01_MSTR_l
  bone2Name=spine_03
  swing1LimitDegrees=45
  swing1Motion=Free
  swing2LimitDegrees=45
  swing2Motion=Free
  twistLimitDegrees=45
  twistMotion=Free
 }
 {
  bone1Name=spine_02
  bone2Name=spine_01
  swing1LimitDegrees=45
  swing1Motion=Free
  swing2LimitDegrees=45
  swing2Motion=Free
  twistLimitDegrees=45
  twistMotion=Free
 }
 {
  bone1Name=spine_03
  bone2Name=spine_02
  swing1LimitDegrees=45
  swing1Motion=Free
  swing2LimitDegrees=45
  swing2Motion=Free
  twistLimitDegrees=45
  twistMotion=Free
 }
 {
  bone1Name=spine_01
  bone2Name=pelvis
  swing1LimitDegrees=45
  swing1Motion=Free
  swing2LimitDegrees=45
  swing2Motion=Free
  twistLimitDegrees=45
  twistMotion=Free
 }
 {
  bone1Name=thigh_l
  bone2Name=pelvis
  swing1LimitDegrees=45
  swing1Motion=Free
  swing2LimitDegrees=45
  swing2Motion=Free
  twistLimitDegrees=45
  twistMotion=Free
 }
 {
  bone1Name=calf_l
  bone2Name=thigh_l
  swing1LimitDegrees=45
  swing1Motion=Free
  swing2LimitDegrees=45
  swing2Motion=Free
  twistLimitDegrees=45
  twistMotion=Free
 }
 {
  bone1Name=thigh_r
  bone2Name=pelvis
  swing1LimitDegrees=45
  swing1Motion=Free
  swing2LimitDegrees=45
  swing2Motion=Free
  twistLimitDegrees=45
  twistMotion=Free
 }
 {
  bone1Name=calf_r
  bone2Name=thigh_r
  swing1LimitDegrees=45
  swing1Motion=Free
  swing2LimitDegrees=45
  swing2Motion=Free
  twistLimitDegrees=45
  twistMotion=Free
 }
 {
  bone1Name=foot_l
  bone2Name=calf_l
  swing1LimitDegrees=45
  swing1Motion=Free
  swing2LimitDegrees=45
  swing2Motion=Free
  twistLimitDegrees=45
  twistMotion=Free
 }
 {
  bone1Name=foot_r
  bone2Name=calf_r
  swing1LimitDegrees=45
  swing1Motion=Free
  swing2LimitDegrees=45
  swing2Motion=Free
  twistLimitDegrees=45
  twistMotion=Free
 }
 {bone1Name=dyn_SkirtE03_l, bone2Name=dyn_SkirtE01_l, swing1LimitDegrees=20, swing2LimitDegrees=20, twistLimitDegrees=0}
 {bone1Name=dyn_SkirtF03_c, bone2Name=dyn_SkirtF01_c, swing1LimitDegrees=20, swing2LimitDegrees=20, twistLimitDegrees=0}
 {bone1Name=dyn_SkirtE03_r, bone2Name=dyn_SkirtE01_r, swing1LimitDegrees=20, swing2LimitDegrees=20, twistLimitDegrees=0}
]
