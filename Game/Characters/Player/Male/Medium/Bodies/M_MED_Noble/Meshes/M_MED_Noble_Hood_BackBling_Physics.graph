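# Physics asset for the M_MED_Noble male hood back bling, giving capsule bodies to the hood, neck collar and upper body.
# Hood back, top and front chains hang mostly from small MSTR root bodies, which join spine_05 with free angular constraints, as do the spine and arm bodies.
# Most chain links and the collar use limited swing with zero twist.
# The front-right chain differs: it has no MSTR root and is angularly free throughout.
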PhysicsAsset /Game/Characters/Player/Male/Medium/Bodies/M_MED_Noble/Meshes/M_MED_Noble_Hood_BackBling_Physics
bodies {
 clavicle_l [{shapeType=Capsule, length=9.4322, radius=8.054, shapeName=clavicle_l_capsule, center=<7.0118 -1.5992 -2.0851>, rotation=<rot 90 0 0>}]
 clavicle_r [{
   shapeType=Capsule
   length=10.4984
   radius=6.9877
   shapeName=clavicle_l_capsule
   center=<-6.9879 0.3791 1.37>
   rotation=<rot 90 180 -176.3848>
  }]
 dyn_hoodBack_MSTR_c_01 [{shapeType=Capsule, length=1, radius=1, shapeName=dyn_hoodBack_MSTR_c_01_capsule}]
 dyn_hoodBack_MSTR_l_01 [{shapeType=Capsule, length=1, radius=1, shapeName=dyn_hoodBack_MSTR_l_01_capsule}]
 dyn_hoodBack_MSTR_r_01 [{shapeType=Capsule, length=1, radius=1, shapeName=dyn_hoodBack_MSTR_r_01_capsule, rotation=<rot 0 -90 0>}]
 dyn_hoodBack_c_01 [{shapeType=Capsule, length=8.6798, radius=2.3159, shapeName=dyn_hoodBack_c_01_capsule, center=<5.333 0 0>}]
 dyn_hoodBack_c_02 [{shapeType=Capsule, length=6.6823, radius=3.3533, shapeName=dyn_hoodBack_c_02_capsule, center=<4.6634 0.7 0>, rotation=<rot 90 0 0>}]
 dyn_hoodBack_c_03 [{shapeType=Capsule, length=6.3939, radius=4.6995, shapeName=dyn_hoodBack_c_03_capsule, center=<4.0413 0 0>}]
 dyn_hoodBack_l_01 [{
   shapeType=Capsule
   length=8.4077
   radius=1.6553
   shapeName=dyn_hoodBack_l_01_capsule
   center=<10.5744 0.0091 -0.0313>
   rotation=<rot 73.7505 -87.1328 -83.319>
  }]
 dyn_hoodBack_l_02 [{shapeType=Capsule, length=10.1073, radius=2.8573, shapeName=dyn_hoodBack_l_02_capsule, center=<8.2 -1.7 -0.7>, rotation=<rot -12 5 -33>}]
 dyn_hoodBack_r_01 [{shapeType=Capsule, length=9.3126, radius=2.0705, shapeName=dyn_hoodBack_r_01_capsule, center=<10.8215 0 0>, rotation=<rot 90 0 0>}]
 dyn_hoodBack_r_02 [{
   shapeType=Capsule
   length=10.0789
   radius=2.7019
   shapeName=dyn_hoodBack_r_02_capsule
   center=<7.6 -2.2 1.4>
   rotation=<rot 6.7168 4.2737 32.5758>
  }]
 dyn_hoodFrontA_l_01 [{shapeType=Capsule, length=6.7161, radius=1.3246, shapeName=dyn_hoodFrontA_l_01_capsule, center=<4.2282 0 0>, rotation=<rot 90 0 0>}]
 dyn_hoodFrontA_l_02 [
  {shapeType=Capsule, length=6.8316, radius=1.5331, shapeName=dyn_hoodFrontA_l_02_capsule, center=<6.4725 0 0>, rotation=<rot 90 0 0>}
  {shapeType=Capsule, length=5, radius=1.3479, shapeName=dyn_hoodFrontA_l_02_capsule, center=<6.1362 0.4578 4.5285>, rotation=<rot 37.374 0 0>}
 ]
 dyn_hoodFrontA_r_01 [{shapeType=Capsule, length=1, radius=1, shapeName=dyn_hoodFrontA_r_01_capsule}]
 dyn_hoodFrontA_r_02 [{shapeType=Capsule, length=8.5713, radius=1.4287, shapeName=dyn_hoodFrontA_r_02_capsule, center=<3.9708 0 0>, rotation=<rot 90 0 0>}]
 dyn_hoodFrontB_l_01 [{shapeType=Capsule, length=6.742, radius=1.2987, shapeName=dyn_hoodFrontB_l_01_capsule, center=<4.2282 0 0>, rotation=<rot 90 0 0>}]
 dyn_hoodTop_MSTR_l_01 [{shapeType=Capsule, length=1, radius=1, shapeName=dyn_hoodTop_MSTR_l_01_capsule}]
 dyn_hoodTop_MSTR_r_01 [{shapeType=Capsule, length=1, radius=1, shapeName=dyn_hoodTop_MSTR_r_01_capsule, rotation=<rot 0 -90 0>}]
 dyn_hoodTop_l_01 [{
   shapeType=Capsule
   length=6.3049
   radius=1.7358
   shapeName=dyn_hoodTop_l_01_capsule
   center=<4.2207 0.5 0>
   rotation=<rot 90 -167.0926 -170.8471>
  }]
 dyn_hoodTop_l_02 [{shapeType=Capsule, length=6.3049, radius=1.7358, shapeName=dyn_hoodTop_l_02_capsule, center=<4.5827 0.5 0>, rotation=<rot 90 0 13.5678>}]
 dyn_hoodTop_r_01 [{
   shapeType=Capsule
   length=7.7613
   radius=2.496
   shapeName=dyn_hoodTop_r_01_capsule
   center=<4.8609 0.8327 0>
   rotation=<rot 90 178.2101 168.7294>
  }]
 dyn_hoodTop_r_02 [{shapeType=Capsule, length=7.4343, radius=2.2964, shapeName=dyn_hoodTop_r_02_capsule, center=<4.0538 0.5 0>, rotation=<rot 90 0 12.5116>}]
 dyn_hood_MSTR_l_01 [{shapeType=Capsule, length=1, radius=1, shapeName=dyn_hood_MSTR_l_01_capsule}]
 dyn_neck_collar_1 [{shapeType=Capsule, length=3.0999, radius=3.8254, shapeName=dyn_neck_collar_1_capsule, center=<6.7902 0 0>}]
 spine_03 [{shapeType=Capsule, length=17.1034, radius=13.535, shapeName=spine_04_capsule}]
 spine_04 [{shapeType=Capsule, length=18.4979, radius=12.1405, shapeName=spine_04_capsule, center=<0 -0.4224 0>}]
 spine_05 [
  {shapeType=Capsule, length=22.5862, radius=8.4259, shapeName=spine_05_capsule, center=<0 -1.6294 0>}
  {shapeType=Capsule, length=4.226, radius=13.8282, shapeName=spine_05_capsule, center=<29.3943 -6.8718 0>}
 ]
 upperarm_l [{
   shapeType=Capsule
   length=23.3411
   radius=7.8816
   shapeName=upperarm_l_capsule
   center=<13.4111 -0.7642 -1.1046>
   rotation=<rot 90 -90 -91.8381>
  }]
 upperarm_r [{
   shapeType=Capsule
   length=19.9004
   radius=7.5087
   shapeName=upperarm_l_capsule
   center=<-13.444 0.1866 1.4241>
   rotation=<rot 90 -179.7016 177.5259>
  }]
}
constraints [
 {
  bone1Name=clavicle_l
  bone2Name=spine_05
  swing1LimitDegrees=45
  swing1Motion=Free
  swing2LimitDegrees=45
  swing2Motion=Free
  twistLimitDegrees=45
  twistMotion=Free
 }
 {
  bone1Name=upperarm_l
  bone2Name=clavicle_l
  swing1LimitDegrees=45
  swing1Motion=Free
  swing2LimitDegrees=45
  swing2Motion=Free
  twistLimitDegrees=45
  twistMotion=Free
 }
 {
  bone1Name=clavicle_r
  bone2Name=spine_05
  swing1LimitDegrees=45
  swing1Motion=Free
  swing2LimitDegrees=45
  swing2Motion=Free
  twistLimitDegrees=45
  twistMotion=Free
 }
 {
  bone1Name=upperarm_r
  bone2Name=clavicle_r
  swing1LimitDegrees=45
  swing1Motion=Free
  swing2LimitDegrees=45
  swing2Motion=Free
  twistLimitDegrees=45
  twistMotion=Free
 }
 {
  bone1Name=dyn_hood_MSTR_l_01
  bone2Name=spine_05
  swing1LimitDegrees=45
  swing1Motion=Free
  swing2LimitDegrees=45
  swing2Motion=Free
  twistLimitDegrees=45
  twistMotion=Free
 }
 {bone1Name=dyn_hoodFrontA_l_01, bone2Name=dyn_hood_MSTR_l_01, swing1LimitDegrees=30, swing2LimitDegrees=30, twistLimitDegrees=0}
 {bone1Name=dyn_hoodFrontA_l_02, bone2Name=dyn_hoodFrontA_l_01, swing1LimitDegrees=30, swing2LimitDegrees=30, twistLimitDegrees=0}
 {bone1Name=dyn_hoodFrontB_l_01, bone2Name=dyn_hood_MSTR_l_01, swing1LimitDegrees=30, swing2LimitDegrees=30, twistLimitDegrees=0}
 {
  bone1Name=dyn_hoodTop_MSTR_l_01
  bone2Name=spine_05
  swing1LimitDegrees=45
  swing1Motion=Free
  swing2LimitDegrees=45
  swing2Motion=Free
  twistLimitDegrees=45
  twistMotion=Free
 }
 {bone1Name=dyn_hoodTop_l_01, bone2Name=dyn_hoodTop_MSTR_l_01, swing1LimitDegrees=30, swing2LimitDegrees=30, twistLimitDegrees=0}
 {bone1Name=dyn_hoodTop_l_02, bone2Name=dyn_hoodTop_l_01, swing1LimitDegrees=30, swing2LimitDegrees=30, twistLimitDegrees=0}
 {
  bone1Name=dyn_hoodBack_MSTR_l_01
  bone2Name=spine_05
  swing1LimitDegrees=45
  swing1Motion=Free
  swing2LimitDegrees=45
  swing2Motion=Free
  twistLimitDegrees=45
  twistMotion=Free
 }
 {bone1Name=dyn_hoodBack_l_01, bone2Name=dyn_hoodBack_MSTR_l_01, swing1LimitDegrees=25, swing2LimitDegrees=15, twistLimitDegrees=0}
 {bone1Name=dyn_hoodBack_l_02, bone2Name=dyn_hoodBack_l_01, swing1LimitDegrees=20, swing2LimitDegrees=10, twistLimitDegrees=0}
 {
  bone1Name=dyn_hoodBack_MSTR_c_01
  bone2Name=spine_05
  swing1LimitDegrees=45
  swing1Motion=Free
  swing2LimitDegrees=45
  swing2Motion=Free
  twistLimitDegrees=45
  twistMotion=Free
 }
 {bone1Name=dyn_hoodBack_c_01, bone2Name=dyn_hoodBack_MSTR_c_01, swing1LimitDegrees=40, swing2LimitDegrees=20, twistLimitDegrees=0}
 {bone1Name=dyn_hoodBack_c_02, bone2Name=dyn_hoodBack_c_01, swing1LimitDegrees=40, swing2LimitDegrees=20, twistLimitDegrees=0}
 {bone1Name=dyn_hoodBack_c_03, bone2Name=dyn_hoodBack_c_02, swing1LimitDegrees=40, swing2LimitDegrees=12, twistLimitDegrees=0}
 {
  bone1Name=spine_05
  bone2Name=spine_04
  swing1LimitDegrees=45
  swing1Motion=Free
  swing2LimitDegrees=45
  swing2Motion=Free
  twistLimitDegrees=45
  twistMotion=Free
 }
 {
  bone1Name=spine_04
  bone2Name=spine_03
  swing1LimitDegrees=45
  swing1Motion=Free
  swing2LimitDegrees=45
  swing2Motion=Free
  twistLimitDegrees=45
  twistMotion=Free
 }
 {
  bone1Name=dyn_hoodFrontA_r_01
  bone2Name=spine_05
  swing1LimitDegrees=45
  swing1Motion=Free
  swing2LimitDegrees=45
  swing2Motion=Free
  twistLimitDegrees=45
  twistMotion=Free
 }
 {
  bone1Name=dyn_hoodFrontA_r_02
  bone2Name=dyn_hoodFrontA_r_01
  swing1LimitDegrees=45
  swing1Motion=Free
  swing2LimitDegrees=45
  swing2Motion=Free
  twistLimitDegrees=45
  twistMotion=Free
 }
 {bone1Name=dyn_hoodBack_r_01, bone2Name=dyn_hoodBack_MSTR_r_01, swing1LimitDegrees=20, swing2LimitDegrees=10, twistLimitDegrees=0}
 {bone1Name=dyn_hoodBack_r_02, bone2Name=dyn_hoodBack_r_01, swing1LimitDegrees=20, swing2LimitDegrees=12, twistLimitDegrees=0}
 {bone1Name=dyn_hoodTop_r_01, bone2Name=dyn_hoodTop_MSTR_r_01, swing1LimitDegrees=30, swing2LimitDegrees=20, twistLimitDegrees=0}
 {bone1Name=dyn_hoodTop_r_02, bone2Name=dyn_hoodTop_r_01, swing1LimitDegrees=30, swing2LimitDegrees=20, twistLimitDegrees=0}
 {bone1Name=dyn_neck_collar_1, bone2Name=spine_05, swing1LimitDegrees=45, swing2LimitDegrees=45, twistLimitDegrees=0}
 {
  bone1Name=dyn_hoodTop_MSTR_r_01
  bone2Name=spine_05
  swing1LimitDegrees=45
  swing1Motion=Free
  swing2LimitDegrees=45
  swing2Motion=Free
  twistLimitDegrees=45
  twistMotion=Free
 }
 {
  bone1Name=dyn_hoodBack_MSTR_r_01
  bone2Name=spine_05
  swing1LimitDegrees=45
  swing1Motion=Free
  swing2LimitDegrees=45
  swing2Motion=Free
  twistLimitDegrees=45
  twistMotion=Free
 }
]
