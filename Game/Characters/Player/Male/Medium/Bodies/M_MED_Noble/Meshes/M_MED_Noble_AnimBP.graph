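# Post-process style AnimBP for the Noble male player body, adding deformation and secondary motion to cloak, hood and sleeves.
# The parent mesh pose passes through two Control Rigs and bone fixes into switched RigidBody cloak sims (default, front end, skydive, crouch).
# Hood sims (with backbling variant) follow, then AnimDynamics sleeves and state-driven offsets, with extra corrections for custom emotes.
# The EventGraph reads the owning player's anim instance and montage for state.
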AnimBlueprint /Game/Characters/Player/Male/Medium/Bodies/M_MED_Noble/Meshes/M_MED_Noble_AnimBP
target_skeleton=&/M_MED_Noble_Skeleton
graphs [
 AnimGraph {nodes [
   "Constraint - hand_r_Sleeves_03 (P)"(Alpha=0.8, ConstraintWeights_0=1) {
    alphaBoolBlend {blendOption=Linear}
    boneToModify {boneName=hand_r_Sleeves_03}
    constraintSetup [{targetBone {boneName=hand_r}}]
    constraintWeights [1]
   }
   n1 = "Transform (Modify) Bone - Bone: dyn_hood_MSTR_r_01"(
    ComponentPose=Local To Component(LocalPose=Use cached pose 'finalCache'())
    Alpha=1
    Translation="2.553,1.332,-8.627"
    Rotation="-0.081,-7.987,12.780"
    Scale="1,1,1"
   ) {
    rotationMode=BMM_Additive
    rotationSpace=BCS_BoneSpace
    translationMode=BMM_Additive
    translationSpace=BCS_BoneSpace
    alphaBoolBlend {blendOption=Linear}
    boneToModify {boneName=dyn_hood_MSTR_r_01}
    rotation=<rot -0.081 -7.987 12.78>
    translation=<2.553 1.332 -8.627>
   }
   n2 = "Transform (Modify) Bone - Bone: dyn_hood_MSTR_l_01"(ComponentPose=@n1, Alpha=1, Translation="3.200,-5.675,3.191", Rotation="10.308,-28.481,-29.061", Scale="1,1,1") {
    rotationMode=BMM_Additive
    rotationSpace=BCS_BoneSpace
    translationMode=BMM_Additive
    translationSpace=BCS_BoneSpace
    alphaBoolBlend {blendOption=Linear}
    boneToModify {boneName=dyn_hood_MSTR_l_01}
    rotation=<rot 10.308 -28.481 -29.061>
    translation=<3.2 -5.675 3.191>
   }
   n3 = "Transform (Modify) Bone - Bone: upperarmDn_r"(ComponentPose=@n2, Alpha=1, Translation="-20,0,0", Rotation="0,0,0", Scale="1,1,1") {
    rotationMode=BMM_Additive
    rotationSpace=BCS_BoneSpace
    translationMode=BMM_Additive
    translationSpace=BCS_BoneSpace
    alphaBoolBlend {blendOption=Linear}
    boneToModify {boneName=upperarmDn_r}
    translation {x=-20}
   }
   n4 = "Transform (Modify) Bone - Bone: dyn_SkirtB01_MSTR_r"(ComponentPose=@n3, Alpha=1, Translation="0,2,0", Rotation="0,-5,0", Scale="1,1,1") {
    rotationMode=BMM_Additive
    rotationSpace=BCS_BoneSpace
    translationMode=BMM_Additive
    translationSpace=BCS_BoneSpace
    alphaBoolBlend {blendOption=Linear}
    boneToModify {boneName=dyn_SkirtB01_MSTR_r}
    rotation {yaw=-5}
    translation {y=2}
   }
   n5 = "Transform (Modify) Bone - Bone: hand_l_Sleeves_02"(ComponentPose=@n4, Alpha=0, Translation="0,0,0", Rotation="0,0,0", Scale="1,1,0.5") {
    rotationMode=BMM_Additive
    rotationSpace=BCS_BoneSpace
    scaleMode=BMM_Additive
    scaleSpace=BCS_BoneSpace
    translationMode=BMM_Additive
    translationSpace=BCS_BoneSpace
    alphaBoolBlend {blendOption=Linear}
    boneToModify {boneName=hand_l_Sleeves_02}
    scale {z=0.5}
   }
   n6 = "Transform (Modify) Bone - Bone: hand_l_Sleeves_04"(ComponentPose=@n5, Alpha=1, Translation="1.430,-0.588,-2.459", Rotation="-0.005,0.011,21.97", Scale="1.000,1.5,1.000") {
    rotationMode=BMM_Additive
    rotationSpace=BCS_BoneSpace
    scaleMode=BMM_Additive
    scaleSpace=BCS_BoneSpace
    translationMode=BMM_Additive
    translationSpace=BCS_BoneSpace
    alphaBoolBlend {blendOption=Linear}
    boneToModify {boneName=hand_l_Sleeves_04}
    rotation=<rot -0.005 0.011 21.97>
    scale {y=1.5}
    translation=<1.43 -0.588 -2.459>
   }
   n7 = "Transform (Modify) Bone - Bone: hand_r_Sleeves_01"(ComponentPose=@n6, Alpha=0, Translation="1.338,-0.092,-6.722", Rotation="0,0,0", Scale="1,1,0.1") {
    rotationMode=BMM_Additive
    rotationSpace=BCS_BoneSpace
    scaleMode=BMM_Additive
    scaleSpace=BCS_BoneSpace
    translationMode=BMM_Additive
    translationSpace=BCS_BoneSpace
    alphaBoolBlend {blendOption=Linear}
    boneToModify {boneName=hand_r_Sleeves_01}
    scale {z=0.1}
    translation=<1.338 -0.092 -6.722>
   }
   n8 = "Transform (Modify) Bone - Bone: hand_r_Sleeves_03"(ComponentPose=@n7, Alpha=1, Translation="6.525,9.298,1.478", Rotation="0,0,0", Scale="1,1,1") {
    rotationMode=BMM_Additive
    rotationSpace=BCS_BoneSpace
    scaleMode=BMM_Additive
    scaleSpace=BCS_BoneSpace
    translationMode=BMM_Additive
    translationSpace=BCS_BoneSpace
    alphaBoolBlend {blendOption=Linear}
    boneToModify {boneName=hand_r_Sleeves_03}
    translation=<6.525 9.298 1.478>
   }
   n9 = "Transform (Modify) Bone - Bone: lowerarm_twist_01_l"(ComponentPose=@n8, Alpha=1, Translation="-4,0,0", Rotation="0,0,0", Scale="1,0.5,1") {
    rotationMode=BMM_Additive
    rotationSpace=BCS_BoneSpace
    scaleMode=BMM_Additive
    scaleSpace=BCS_BoneSpace
    translationMode=BMM_Additive
    translationSpace=BCS_BoneSpace
    alphaBoolBlend {blendOption=Linear}
    boneToModify {boneName=lowerarm_twist_01_l}
    scale {y=0.5}
    translation {x=-4}
   }
   n10 = "Transform (Modify) Bone - Bone: lowerarm_twist_02_l"(ComponentPose=@n9, Alpha=1, Translation="0,0,0", Rotation="0,0,0", Scale="1,0.5,1") {
    rotationMode=BMM_Additive
    rotationSpace=BCS_BoneSpace
    scaleMode=BMM_Additive
    scaleSpace=BCS_BoneSpace
    translationMode=BMM_Additive
    translationSpace=BCS_BoneSpace
    alphaBoolBlend {blendOption=Linear}
    boneToModify {boneName=lowerarm_twist_02_l}
    scale {y=0.5}
   }
   n11 = "Transform (Modify) Bone - Bone: lowerarm_twist_01_r"(ComponentPose=@n10, Alpha=1, Translation="0,0,0", Rotation="0,0,0", Scale="1,0.5,1") {
    rotationMode=BMM_Additive
    rotationSpace=BCS_BoneSpace
    scaleMode=BMM_Additive
    scaleSpace=BCS_BoneSpace
    translationMode=BMM_Additive
    translationSpace=BCS_BoneSpace
    alphaBoolBlend {blendOption=Linear}
    boneToModify {boneName=lowerarm_twist_01_r}
    scale {y=0.5}
   }
   n12 = "Transform (Modify) Bone - Bone: lowerarm_twist_02_r"(ComponentPose=@n11, Alpha=1, Translation="0,0,0", Rotation="0,0,45", Scale="1,0.5,1") {
    rotationMode=BMM_Additive
    rotationSpace=BCS_BoneSpace
    scaleMode=BMM_Additive
    scaleSpace=BCS_BoneSpace
    translationMode=BMM_Additive
    translationSpace=BCS_BoneSpace
    alphaBoolBlend {blendOption=Linear}
    boneToModify {boneName=lowerarm_twist_02_r}
    rotation {roll=45}
    scale {y=0.5}
   }
   Output Pose(Result=Blend Poses by bool(
     BlendPose_0=Component To Local(ComponentPose=@n12)
     BlendPose_1=Use cached pose 'finalCache'()
     bActiveValue=Get isCustomEmotePlaying() -> (isCustomEmotePlaying:bool)
    ) {blendType=Linear}) {layerGroup=None}
   n13 = Control Rig(Source=Copy Pose From Mesh() {bUseAttachedParent=true}) {
    alphaBoolBlend {blendOption=Linear}
    controlRigAssetReference {blueprintRigClass=&/M_MED_Noble_Deform_CtrlRig_twists.M_MED_Noble_Deform_CtrlRig_twists_C}
   }
   n14 = Control Rig!(
    Source=@n13
    Alpha
    KneeRot_fr_l_114deg:Vector
    KneeTrans_fr_l_114deg:Vector="0.258,-3.591,0.009"
    KneeScale_fr_l_114deg:Vector="1,1,1"
    KneeScale_fr_l_130deg:Vector="1,1,1"
    KneeRot_fr_l_130deg:Vector="0,0,-51.2"
    KneeTrans_fr_l_130deg:Vector="0.258,-3.591,0.009"
   ) -> (Pose) {
    alphaBoolBlend {blendOption=Linear}
    controlRigAssetReference {blueprintRigClass=&/M_MED_Noble_Deform_CtrlRig_sculpters.M_MED_Noble_Deform_CtrlRig_sculpters_C}
   }
   Save cached pose 'controlRigCache'(Pose=@n14) {cacheName=controlRigCache}
   n15 = "Transform (Modify) Bone - Bone: hand_l_Sleeves_02"!(
    ComponentPose=Local To Component(LocalPose=Use cached pose 'RBANCache'())
    bAlphaBoolEnabled:bool=false
    Translation="-3.333,24.381,3.569"
    Rotation="-0.071,2.989,-0.064"
    Scale="1,1,1"
   ) -> (Pose) {
    alphaInputType=Bool
    rotationMode=BMM_Additive
    rotationSpace=BCS_BoneSpace
    translationMode=BMM_Additive
    translationSpace=BCS_BoneSpace
    alphaBoolBlend {blendOption=Linear}
    boneToModify {boneName=hand_l_Sleeves_02}
    rotation=<rot -0.071 2.989 -0.064>
    translation=<-3.333 24.381 3.569>
   }
   n16 = Anim Dynamics - Bone: hand_l_Sleeves_02!(ComponentPose=@n15, bAlphaBoolEnabled:bool=Get isRifle() -> (isRifle:bool), ExternalForce="0,0,0") -> (Pose) {
    alphaInputType=Bool
    angularDampingOverride=0.9
    angularSpringConstant=1000
    bAngularSpring=true
    bLinearSpring=true
    bOverrideAngularDamping=true
    bOverrideLinearDamping=true
    bUseGravityOverride=true
    bUseSphericalLimits=true
    gravityScale=0.35
    linearDampingOverride=0.9999
    linearSpringConstant=75
    simulationSpace=BoneRelative
    alphaBoolBlend {blendOption=Linear}
    boundBone {boneName=hand_l_Sleeves_02}
    componentAppliedLinearAccClamp=<200 200 1000>
    componentLinearAccScale=<3 3 8>
    componentLinearVelScale=<0.1 0.1 0.5>
    gravityOverride {z=-980}
    physicsBodyDefinitions [{
      boundBone {boneName=hand_l_Sleeves_02}
      boxExtents=<20 10 20>
      constraintSetup {angularLimitsMax=<0 0 0>, angularLimitsMin=<0 0 0>, angularTarget=<0 0 0>, linearAxesMax=<2 2 2>, linearAxesMin=<-2 -2 -2>}
      localJointOffset=<-5 -10 0>
     }]
    relativeSpaceBone {boneName=root}
   }
   n17 = Anim Dynamics - Bone: hand_r_Sleeves_01(ComponentPose=@n16, Alpha=1, ExternalForce="0,0,0") {
    angularDampingOverride=0.9
    angularSpringConstant=1000
    bAngularSpring=true
    bLinearSpring=true
    bOverrideAngularDamping=true
    bOverrideLinearDamping=true
    bUseGravityOverride=true
    bUseSphericalLimits=true
    gravityScale=0.35
    linearDampingOverride=0.9999
    linearSpringConstant=75
    simulationSpace=BoneRelative
    alphaBoolBlend {blendOption=Linear}
    boundBone {boneName=hand_r_Sleeves_01}
    componentAppliedLinearAccClamp=<200 200 1000>
    componentLinearAccScale=<3 3 8>
    componentLinearVelScale=<0.1 0.1 0.5>
    gravityOverride {z=-980}
    physicsBodyDefinitions [{
      boundBone {boneName=hand_r_Sleeves_01}
      boxExtents=<20 10 20>
      constraintSetup {angularLimitsMax=<0 0 0>, angularLimitsMin=<0 0 0>, angularTarget=<0 0 0>, linearAxesMax=<2 2 2>, linearAxesMin=<-2 -2 -2>}
      localJointOffset=<5 10 0>
     }]
    relativeSpaceBone {boneName=root}
   }
   n18 = "Transform (Modify) Bone - Bone: hand_r_Sleeves_01"!(ComponentPose=@n17, bAlphaBoolEnabled:bool=Get isMelee() -> (isMelee:bool), Translation="20,-5,-10", Rotation="0,30,0", Scale="1,1,1") -> (Pose) {
    alphaInputType=Bool
    rotationMode=BMM_Additive
    translationMode=BMM_Additive
    translationSpace=BCS_BoneSpace
    alphaBoolBlend {blendOption=Linear}
    boneToModify {boneName=hand_r_Sleeves_01}
    rotation {yaw=30}
    translation=<20 -5 -10>
   }
   n19 = "Transform (Modify) Bone - Bone: hand_r_Sleeves_01"(
    ComponentPose=@n18
    bAlphaBoolEnabled=Get PlayerMovement() -> (PlayerMovement:bool)
    Translation="9,-2,-5"
    Rotation="0,0,0"
    Scale="0.2,0.2,0.3"
   ) {
    alphaInputType=Bool
    rotationMode=BMM_Replace
    scaleMode=BMM_Additive
    scaleSpace=BCS_BoneSpace
    translationMode=BMM_Additive
    translationSpace=BCS_BoneSpace
    alphaBoolBlend {blendInTime=1, blendOption=Linear, blendOutTime=0.1}
    boneToModify {boneName=hand_r_Sleeves_01}
    scale=<0.2 0.2 0.3>
    translation=<9 -2 -5>
   }
   n20 = "Transform (Modify) Bone - Bone: hand_r_Sleeves_03"!(
    ComponentPose=@n19
    bAlphaBoolEnabled:bool=Get Alt_weapon() -> (Alt_weapon:bool)
    Translation="4.244,2.435,2.021"
    Rotation="0,0,0"
    Scale="1,1,1"
   ) -> (Pose) {
    alphaInputType=Bool
    rotationMode=BMM_Replace
    scaleMode=BMM_Additive
    scaleSpace=BCS_BoneSpace
    translationMode=BMM_Additive
    translationSpace=BCS_BoneSpace
    alphaBoolBlend {blendInTime=1, blendOption=Linear, blendOutTime=0.1}
    boneToModify {boneName=hand_r_Sleeves_03}
    translation=<4.244 2.435 2.021>
   }
   n21 = Get Alt_weapon()
   n22 = "Transform (Modify) Bone - Bone: hand_l_Sleeves_02"(ComponentPose=@n20, bAlphaBoolEnabled=@n21, Translation="-20,25,-20", Rotation="5.315,0.082,-0.097", Scale="1,1,1") {
    alphaInputType=Bool
    rotationMode=BMM_Additive
    rotationSpace=BCS_BoneSpace
    translationMode=BMM_Additive
    translationSpace=BCS_BoneSpace
    alphaBoolBlend {blendOption=Linear}
    boneToModify {boneName=hand_l_Sleeves_02}
    rotation=<rot 5.315 0.082 -0.097>
    translation=<-20 25 -20>
   }
   n23 = "Transform (Modify) Bone - Bone: hand_l_Sleeves_04"!(ComponentPose=@n22, bAlphaBoolEnabled:bool=@n21, Translation="0.136,0.147,2.628", Rotation="3.990,17.468,14.511", Scale="1,1,1") -> (Pose) {
    alphaInputType=Bool
    rotationMode=BMM_Additive
    rotationSpace=BCS_BoneSpace
    translationMode=BMM_Additive
    translationSpace=BCS_BoneSpace
    alphaBoolBlend {blendOption=Linear}
    boneToModify {boneName=hand_l_Sleeves_04}
    rotation=<rot 3.99 17.468 14.511>
    translation=<0.136 0.147 2.628>
   }
   n24 = Anim Dynamics - Bone: hand_l_Sleeves_04!(ComponentPose=@n23, bAlphaBoolEnabled:bool=@n21, ExternalForce="0,0,0") -> (Pose) {
    alphaInputType=Bool
    angularDampingOverride=0.9
    angularSpringConstant=1000
    bAngularSpring=true
    bLinearSpring=true
    bOverrideAngularDamping=true
    bOverrideLinearDamping=true
    bUseGravityOverride=true
    bUseSphericalLimits=true
    gravityScale=0.35
    linearDampingOverride=0.9999
    linearSpringConstant=80
    simulationSpace=BoneRelative
    alphaBoolBlend {blendOption=Linear}
    boundBone {boneName=hand_l_Sleeves_04}
    componentAppliedLinearAccClamp=<200 200 1000>
    componentLinearAccScale=<3 3 8>
    componentLinearVelScale=<0.1 0.1 0.5>
    gravityOverride {z=-980}
    physicsBodyDefinitions [{
      boundBone {boneName=hand_l_Sleeves_04}
      boxExtents=<15 15 7>
      constraintSetup {angularLimitsMax=<0 0 0>, angularLimitsMin=<0 0 0>, angularTarget=<0 0 0>, linearAxesMax=<1 1 1>, linearAxesMin=<-1 -1 -1>}
      localJointOffset=<0 -1 4>
     }]
    relativeSpaceBone {boneName=root}
   }
   n25 = "Transform (Modify) Bone - Bone: upperarmDn_l"(ComponentPose=@n24, Alpha=Get Head Up ADS() -> (Head Up ADS:double), Translation="0,0,0", Rotation="-80,0,50", Scale="1,1,1") {
    rotationMode=BMM_Additive
    rotationSpace=BCS_BoneSpace
    scaleMode=BMM_Additive
    translationMode=BMM_Additive
    translationSpace=BCS_BoneSpace
    alphaBoolBlend {blendOption=Linear}
    boneToModify {boneName=upperarmDn_l}
    rotation {pitch=-80, roll=50}
   }
   Save cached pose 'finalCache'(Pose=Component To Local(ComponentPose=@n25)) {cacheName=finalCache}
   n26 = "Constraint - upperarmDn_r (PPP)"!(
    ComponentPose=Local To Component(LocalPose=Use cached pose 'controlRigCache'())
    Alpha=1
    ConstraintWeights_0=0.8
    ConstraintWeights_1:float=0.2
    ConstraintWeights_2:float=1
   ) -> (Pose) {
    alphaBoolBlend {blendOption=Linear}
    boneToModify {boneName=upperarmDn_r}
    constraintSetup [{targetBone {boneName=spine_04}}, {targetBone {boneName=clavicle_r}}, {targetBone {boneName=upperarm_twist_02_r}}]
    constraintWeights [1,1,1]
   }
   n27 = "Constraint - upperarmDn_l (PPP)"!(ComponentPose=@n26, Alpha=0, ConstraintWeights_0=0.2, ConstraintWeights_1:float=0.8, ConstraintWeights_2:float=1) -> (Pose) {
    alphaBoolBlend {blendOption=Linear}
    boneToModify {boneName=upperarmDn_l}
    constraintSetup [{targetBone {boneName=spine_04}}, {targetBone {boneName=clavicle_l}}, {targetBone {boneName=upperarm_twist_02_r}}]
    constraintWeights [1,1,1]
   }
   n28 = "Transform (Modify) Bone - Bone: dyn_SkirtB01_MSTR_r"!(ComponentPose=@n27, bAlphaBoolEnabled:bool=Get bIsJumping(), Translation="0,0,0", Rotation="-15,45,60", Scale="1,1,1") -> (Pose) {
    alphaInputType=Bool
    rotationMode=BMM_Additive
    rotationSpace=BCS_BoneSpace
    alphaBoolBlend {blendInTime=1, blendOption=Linear, blendOutTime=3}
    boneToModify {boneName=dyn_SkirtB01_MSTR_r}
    rotation=<rot -15 45 60>
   }
   n29 = "Transform (Modify) Bone - Bone: dyn_SkirtB01_MSTR_l"!(ComponentPose=@n28, bAlphaBoolEnabled:bool=false, Translation="0,0,0", Rotation="15.823,20.003,0.008", Scale="1,1,1") -> (Pose) {
    alphaInputType=Bool
    rotationMode=BMM_Additive
    rotationSpace=BCS_BoneSpace
    alphaBoolBlend {blendInTime=1, blendOption=Linear, blendOutTime=1}
    boneToModify {boneName=dyn_SkirtB01_MSTR_l}
    rotation=<rot 15.823 20.003 0.008>
   }
   n30 = "Constraint - hand_l_Sleeves_04 (P)"(ComponentPose=@n29, Alpha=1, ConstraintWeights_0=1) {
    alphaBoolBlend {blendOption=Linear}
    boneToModify {boneName=hand_l_Sleeves_04}
    constraintSetup [{targetBone {boneName=hand_l}}]
    constraintWeights [1]
   }
   n31 = Get bIsSliding()
   n32 = "Transform (Modify) Bone - Bone: dyn_SkirtB02_r"!(ComponentPose=@n30, bAlphaBoolEnabled:bool=@n31, Translation="0,0,0", Rotation="-20,50,0", Scale="1,1,1") -> (Pose) {
    alphaInputType=Bool
    rotationMode=BMM_Additive
    rotationSpace=BCS_BoneSpace
    scaleMode=BMM_Additive
    translationMode=BMM_Additive
    translationSpace=BCS_BoneSpace
    alphaBoolBlend {blendOption=Linear}
    boneToModify {boneName=dyn_SkirtB02_r}
    rotation {pitch=-20, yaw=50}
   }
   n33 = "Transform (Modify) Bone - Bone: dyn_SkirtB02_l"!(ComponentPose=@n32, bAlphaBoolEnabled:bool=@n31, Translation="0,0,0", Rotation="30,50,-50", Scale="1,1,1") -> (Pose) {
    alphaInputType=Bool
    rotationMode=BMM_Additive
    rotationSpace=BCS_BoneSpace
    scaleMode=BMM_Additive
    translationMode=BMM_Additive
    translationSpace=BCS_BoneSpace
    alphaBoolBlend {blendOption=Linear}
    boneToModify {boneName=dyn_SkirtB02_l}
    rotation=<rot 30 50 -50>
   }
   Save cached pose 'PreSimCache'(Pose=Component To Local(ComponentPose=@n33)) {cacheName=PreSimCache}
   n34 = RigidBody!(
    ComponentPose=Local To Component(LocalPose=Use cached pose 'PreSimCache'())
    Alpha
    OverrideWorldGravity:Vector=Get Gravity Override()
    ExternalForce
   ) -> (Pose) {
    bClampLinearTranslationLimitToRefPose=true
    bOverrideWorldGravity=true
    overridePhysicsAsset=&/M_MED_Noble_Cloak_Physics
    simulationSpace=BaseBoneSpace
    alphaBoolBlend {blendOption=Linear}
    baseBoneRef {boneName=root}
    componentAppliedLinearAccClamp=<1800 1800 1800>
    componentLinearAccScale=<0.01 0.001 0.3>
    componentLinearVelScale=<0.08 0.05 0.1>
    overrideWorldGravity {z=-980}
   }
   n35 = Component To Local(ComponentPose=@n34)
   n36 = RigidBody(ComponentPose=Local To Component(LocalPose=Use cached pose 'PreSimCache'()), OverrideWorldGravity=Get Gravity Override()) {
    bClampLinearTranslationLimitToRefPose=true
    bOverrideWorldGravity=true
    overridePhysicsAsset=&/M_MED_Noble_Cloak_FrontENd_Physics
    simulationSpace=BaseBoneSpace
    alphaBoolBlend {blendOption=Linear}
    baseBoneRef {boneName=root}
    componentAppliedLinearAccClamp=<1800 1800 1800>
    componentLinearAccScale=<0.01 0.001 0.3>
    componentLinearVelScale=<0.08 0.05 0.1>
    overrideWorldGravity {z=-980}
   }
   n37 = Component To Local(ComponentPose=@n36)
   n38 = Get bIsInFrontEnd()
   n39 = Get bIsSprinting()
   n40 = "Transform (Modify) Bone - Bone: dyn_SkirtF04_c"!(
    ComponentPose=Local To Component(LocalPose=Blend Poses by bool(BlendPose_0=@n37, BlendPose_1=@n35, bActiveValue=@n38) {blendType=Linear})
    bAlphaBoolEnabled:bool=@n39
    Translation="0,0,0"
    Rotation="-7,0,0"
    Scale="1,1,1"
   ) -> (Pose) {
    alphaInputType=Bool
    rotationMode=BMM_Additive
    rotationSpace=BCS_BoneSpace
    scaleMode=BMM_Additive
    scaleSpace=BCS_BoneSpace
    alphaBoolBlend {blendInTime=0.2, blendOption=Linear, blendOutTime=1}
    boneToModify {boneName=dyn_SkirtF04_c}
    rotation {pitch=-7}
   }
   n41 = "Transform (Modify) Bone - Bone: dyn_SkirtF06_c"!(ComponentPose=@n40, bAlphaBoolEnabled:bool=@n39, Translation="0,0,0", Rotation="-7,0,0", Scale="0.8,1,1") -> (Pose) {
    alphaInputType=Bool
    rotationMode=BMM_Additive
    rotationSpace=BCS_BoneSpace
    scaleMode=BMM_Additive
    scaleSpace=BCS_BoneSpace
    alphaBoolBlend {blendInTime=0.2, blendOption=Linear, blendOutTime=1}
    boneToModify {boneName=dyn_SkirtF06_c}
    rotation {pitch=-7}
    scale {x=0.8}
   }
   n42 = Get bIsSprinting()
   n43 = "Transform (Modify) Bone - Bone: dyn_SkirtE04_l"!(ComponentPose=@n41, bAlphaBoolEnabled:bool=@n42, Translation="0,0,0", Rotation="3,7,0", Scale="1,1,1") -> (Pose) {
    alphaInputType=Bool
    rotationMode=BMM_Additive
    rotationSpace=BCS_BoneSpace
    scaleMode=BMM_Additive
    scaleSpace=BCS_BoneSpace
    alphaBoolBlend {blendInTime=0.2, blendOption=Linear, blendOutTime=1}
    boneToModify {boneName=dyn_SkirtE04_l}
    rotation {pitch=3, yaw=7}
   }
   n44 = "Transform (Modify) Bone - Bone: dyn_SkirtE06_l"!(ComponentPose=@n43, bAlphaBoolEnabled:bool=@n42, Translation="0,0,0", Rotation="3,7,0", Scale="0.6,1,1") -> (Pose) {
    alphaInputType=Bool
    rotationMode=BMM_Additive
    rotationSpace=BCS_BoneSpace
    scaleMode=BMM_Additive
    scaleSpace=BCS_BoneSpace
    alphaBoolBlend {blendInTime=0.2, blendOption=Linear, blendOutTime=1}
    boneToModify {boneName=dyn_SkirtE06_l}
    rotation {pitch=3, yaw=7}
    scale {x=0.6}
   }
   n45 = Get bIsSprinting()
   n46 = "Transform (Modify) Bone - Bone: dyn_SkirtE04_r"!(ComponentPose=@n44, bAlphaBoolEnabled:bool=@n45, Translation="0,0,0", Rotation="-3,7,0", Scale="1,1,1") -> (Pose) {
    alphaInputType=Bool
    rotationMode=BMM_Additive
    rotationSpace=BCS_BoneSpace
    scaleMode=BMM_Additive
    scaleSpace=BCS_BoneSpace
    alphaBoolBlend {blendInTime=0.2, blendOption=Linear, blendOutTime=1}
    boneToModify {boneName=dyn_SkirtE04_r}
    rotation {pitch=-3, yaw=7}
   }
   n47 = "Transform (Modify) Bone - Bone: dyn_SkirtE06_r"!(ComponentPose=@n46, bAlphaBoolEnabled:bool=@n45, Translation="0,0,0", Rotation="-3,7,0", Scale="0.7,1,1") -> (Pose) {
    alphaInputType=Bool
    rotationMode=BMM_Additive
    rotationSpace=BCS_BoneSpace
    scaleMode=BMM_Additive
    scaleSpace=BCS_BoneSpace
    alphaBoolBlend {blendInTime=0.2, blendOption=Linear, blendOutTime=1}
    boneToModify {boneName=dyn_SkirtE06_r}
    rotation {pitch=-3, yaw=7}
    scale {x=0.7}
   }
   n48 = Component To Local(ComponentPose=@n47)
   n49 = Get bIsSkyDiving()
   n50 = "Transform (Modify) Bone - Bone: hand_r_Sleeves_03"(
    ComponentPose=Local To Component(LocalPose=Use cached pose 'PreSimCache'())
    bAlphaBoolEnabled=@n49
    Translation="10,-15,-10"
    Rotation="50,0,0"
    Scale="1,1,1"
   ) {
    alphaInputType=Bool
    rotationMode=BMM_Additive
    rotationSpace=BCS_BoneSpace
    translationMode=BMM_Additive
    translationSpace=BCS_BoneSpace
    alphaBoolBlend {blendOption=Linear}
    boneToModify {boneName=hand_r_Sleeves_03}
    rotation {pitch=50}
    translation=<10 -15 -10>
   }
   n51 = "Transform (Modify) Bone - Bone: dyn_SkirtB01_MSTR_l"(ComponentPose=@n50, bAlphaBoolEnabled=@n49, Translation="0,0,0", Rotation="15,15,20", Scale="1,1,1") {
    alphaInputType=Bool
    rotationMode=BMM_Additive
    rotationSpace=BCS_BoneSpace
    alphaBoolBlend {blendOption=Linear}
    boneToModify {boneName=dyn_SkirtB01_MSTR_l}
    rotation=<rot 15 15 20>
   }
   n52 = "Transform (Modify) Bone - Bone: dyn_SkirtB01_MSTR_r"(ComponentPose=@n51, bAlphaBoolEnabled=@n49, Translation="0,0,0", Rotation="-10,20,20", Scale="1,1,1") {
    alphaInputType=Bool
    rotationMode=BMM_Additive
    rotationSpace=BCS_BoneSpace
    alphaBoolBlend {blendOption=Linear}
    boneToModify {boneName=dyn_SkirtB01_MSTR_r}
    rotation=<rot -10 20 20>
   }
   n53 = "Transform (Modify) Bone - Bone: dyn_hoodBack_r_02"!(ComponentPose=@n52, bAlphaBoolEnabled:bool=@n49, Translation="10,5,0", Rotation="30,0,20", Scale="1,1,1") -> (Pose) {
    alphaInputType=Bool
    rotationMode=BMM_Additive
    rotationSpace=BCS_BoneSpace
    scaleMode=BMM_Additive
    translationMode=BMM_Additive
    alphaBoolBlend {blendOption=Linear}
    boneToModify {boneName=dyn_hoodBack_r_02}
    rotation {pitch=30, roll=20}
    translation=<10 5>
   }
   n54 = Get bIsParachuteOpened()
   n55 = "Transform (Modify) Bone - Bone: dyn_hoodFrontA_l_01"!(ComponentPose=@n53, bAlphaBoolEnabled:bool=@n54, Translation="0,0,0", Rotation="-60,0,10", Scale="0.6,1,1") -> (Pose) {
    alphaInputType=Bool
    rotationMode=BMM_Additive
    rotationSpace=BCS_BoneSpace
    scaleMode=BMM_Additive
    scaleSpace=BCS_BoneSpace
    alphaBoolBlend {blendOption=Linear}
    boneToModify {boneName=dyn_hoodFrontA_l_01}
    rotation {pitch=-60, roll=10}
    scale {x=0.6}
   }
   n56 = "Transform (Modify) Bone - Bone: dyn_hoodBack_MSTR_l_01"!(ComponentPose=@n55, bAlphaBoolEnabled:bool=@n54, Translation="0,0,0", Rotation="0,10,0", Scale="0.4,1,1") -> (Pose) {
    alphaInputType=Bool
    rotationMode=BMM_Additive
    rotationSpace=BCS_BoneSpace
    scaleMode=BMM_Additive
    scaleSpace=BCS_BoneSpace
    alphaBoolBlend {blendOption=Linear}
    boneToModify {boneName=dyn_hoodBack_MSTR_l_01}
    rotation {yaw=10}
    scale {x=0.4}
   }
   n57 = "Transform (Modify) Bone - Bone: dyn_hoodFrontA_r_01"!(ComponentPose=@n56, bAlphaBoolEnabled:bool=@n54, Translation="0,0,0", Rotation="40,10,0", Scale="0.5,1,0.5") -> (Pose) {
    alphaInputType=Bool
    rotationMode=BMM_Additive
    rotationSpace=BCS_BoneSpace
    scaleMode=BMM_Additive
    scaleSpace=BCS_BoneSpace
    alphaBoolBlend {blendOption=Linear}
    boneToModify {boneName=dyn_hoodFrontA_r_01}
    rotation {pitch=40, yaw=10}
    scale {x=0.5, z=0.5}
   }
   n58 = "Transform (Modify) Bone - Bone: dyn_hoodTop_MSTR_r_01"!(ComponentPose=@n57, bAlphaBoolEnabled:bool=@n54, Translation="0,2,0", Rotation="0,0,0", Scale="1,1,1") -> (Pose) {
    alphaInputType=Bool
    rotationMode=BMM_Additive
    rotationSpace=BCS_BoneSpace
    scaleMode=BMM_Additive
    scaleSpace=BCS_BoneSpace
    translationMode=BMM_Additive
    translationSpace=BCS_BoneSpace
    alphaBoolBlend {blendOption=Linear}
    boneToModify {boneName=dyn_hoodTop_MSTR_r_01}
    translation {y=2}
   }
   n59 = RigidBody(ComponentPose=@n58, OverrideWorldGravity=Get Gravity Override()) {
    bClampLinearTranslationLimitToRefPose=true
    bOverrideWorldGravity=true
    overridePhysicsAsset=&/M_MED_Noble_Cloak_skydive_Physics
    simulationSpace=BaseBoneSpace
    alphaBoolBlend {blendOption=Linear}
    baseBoneRef {boneName=root}
    componentAppliedLinearAccClamp=<1000 1000 200>
    componentLinearAccScale=<0.001 0.001 0.001>
    componentLinearVelScale=<0.001 0.001 0.001>
    overrideWorldGravity {z=-980}
   }
   n60 = Component To Local(ComponentPose=@n59)
   n61 = Get bIsSkyDiving()
   n62 = Get charMoving() -> (charMoving:bool)
   n63 = "Transform (Modify) Bone - Bone: dyn_SkirtB01_MSTR_l"(
    ComponentPose=Local To Component(LocalPose=Use cached pose 'PreSimCache'())
    bAlphaBoolEnabled=@n62
    Translation="-6.7,6.9,14"
    Rotation="7,20,-35"
    Scale="1,1,1"
   ) {
    alphaInputType=Bool
    rotationMode=BMM_Additive
    rotationSpace=BCS_BoneSpace
    scaleMode=BMM_Additive
    scaleSpace=BCS_BoneSpace
    translationMode=BMM_Additive
    translationSpace=BCS_BoneSpace
    alphaBoolBlend {blendInTime=0.2, blendOption=Linear, blendOutTime=0.2}
    boneToModify {boneName=dyn_SkirtB01_MSTR_l}
    rotation=<rot 7 20 -35>
    translation=<-6.7 6.9 14>
   }
   n64 = "Transform (Modify) Bone - Bone: dyn_SkirtB04_l"!(ComponentPose=@n63, bAlphaBoolEnabled:bool=@n62, Translation="0,0,0", Rotation="-6,5,0", Scale="1,1,1") -> (Pose) {
    alphaInputType=Bool
    rotationMode=BMM_Additive
    rotationSpace=BCS_BoneSpace
    scaleMode=BMM_Additive
    scaleSpace=BCS_BoneSpace
    translationMode=BMM_Additive
    translationSpace=BCS_BoneSpace
    alphaBoolBlend {blendInTime=0.2, blendOption=Linear, blendOutTime=0.2}
    boneToModify {boneName=dyn_SkirtB04_l}
    rotation {pitch=-6, yaw=5}
   }
   n65 = "Transform (Modify) Bone - Bone: dyn_SkirtB05_l"!(ComponentPose=@n64, bAlphaBoolEnabled:bool=@n62, Translation="0,0,0", Rotation="0,30,0", Scale="1,1,1") -> (Pose) {
    alphaInputType=Bool
    rotationMode=BMM_Additive
    rotationSpace=BCS_BoneSpace
    scaleMode=BMM_Additive
    scaleSpace=BCS_BoneSpace
    translationMode=BMM_Additive
    translationSpace=BCS_BoneSpace
    alphaBoolBlend {blendInTime=0.2, blendOption=Linear, blendOutTime=0.2}
    boneToModify {boneName=dyn_SkirtB05_l}
    rotation {yaw=30}
   }
   n66 = "Transform (Modify) Bone - Bone: dyn_SkirtD01_MSTR_l"!(ComponentPose=@n65, bAlphaBoolEnabled:bool=@n62, Translation="0.123,-0.122,1.546", Rotation="4.972,14.122,-10.041", Scale="1,1,1") -> (Pose) {
    alphaInputType=Bool
    rotationMode=BMM_Additive
    rotationSpace=BCS_BoneSpace
    translationMode=BMM_Additive
    translationSpace=BCS_BoneSpace
    alphaBoolBlend {blendInTime=0.2, blendOption=Linear, blendOutTime=0.2}
    boneToModify {boneName=dyn_SkirtD01_MSTR_l}
    rotation=<rot 4.972 14.122 -10.041>
    translation=<0.123 -0.122 1.546>
   }
   n67 = "Transform (Modify) Bone - Bone: dyn_SkirtD04_l"!(ComponentPose=@n66, bAlphaBoolEnabled:bool=@n62, Translation="0,0,0", Rotation="0,17.5,0", Scale="1,1,1") -> (Pose) {
    alphaInputType=Bool
    rotationMode=BMM_Additive
    rotationSpace=BCS_BoneSpace
    translationMode=BMM_Additive
    translationSpace=BCS_BoneSpace
    alphaBoolBlend {blendInTime=0.2, blendOption=Linear, blendOutTime=0.2}
    boneToModify {boneName=dyn_SkirtD04_l}
    rotation {yaw=17.5}
   }
   n68 = "Transform (Modify) Bone - Bone: dyn_SkirtD05_l"!(ComponentPose=@n67, bAlphaBoolEnabled:bool=@n62, Translation="0,0,0", Rotation="0,20,0", Scale="1,1,1") -> (Pose) {
    alphaInputType=Bool
    rotationMode=BMM_Additive
    rotationSpace=BCS_BoneSpace
    translationMode=BMM_Additive
    translationSpace=BCS_BoneSpace
    alphaBoolBlend {blendInTime=0.2, blendOption=Linear, blendOutTime=0.2}
    boneToModify {boneName=dyn_SkirtD05_l}
    rotation {yaw=20}
   }
   n69 = "Transform (Modify) Bone - Bone: dyn_SkirtE01_MSTR_l"!(ComponentPose=@n68, bAlphaBoolEnabled:bool=@n62, Translation="0,0,0", Rotation="0,10,0", Scale="1,1,1") -> (Pose) {
    alphaInputType=Bool
    rotationMode=BMM_Additive
    rotationSpace=BCS_BoneSpace
    translationMode=BMM_Additive
    translationSpace=BCS_BoneSpace
    alphaBoolBlend {blendInTime=0.2, blendOption=Linear, blendOutTime=0.2}
    boneToModify {boneName=dyn_SkirtE01_MSTR_l}
    rotation {yaw=10}
   }
   n70 = "Transform (Modify) Bone - Bone: dyn_SkirtE04_l"(ComponentPose=@n69, bAlphaBoolEnabled=@n62, Translation="0,0,0", Rotation="0,10,0", Scale="1,1,1") {
    alphaInputType=Bool
    rotationMode=BMM_Additive
    rotationSpace=BCS_BoneSpace
    translationMode=BMM_Additive
    translationSpace=BCS_BoneSpace
    alphaBoolBlend {blendInTime=0.2, blendOption=Linear, blendOutTime=0.2}
    boneToModify {boneName=dyn_SkirtE04_l}
    rotation {yaw=10}
   }
   n71 = "Transform (Modify) Bone - Bone: dyn_SkirtE05_l"!(ComponentPose=@n70, bAlphaBoolEnabled:bool=@n62, Translation="0,0,0", Rotation="0,10,0", Scale="1,1,1") -> (Pose) {
    alphaInputType=Bool
    rotationMode=BMM_Additive
    rotationSpace=BCS_BoneSpace
    translationMode=BMM_Additive
    translationSpace=BCS_BoneSpace
    alphaBoolBlend {blendInTime=0.2, blendOption=Linear, blendOutTime=0.2}
    boneToModify {boneName=dyn_SkirtE05_l}
    rotation {yaw=10}
   }
   n72 = Get charMoving()
   n73 = "Transform (Modify) Bone - Bone: dyn_SkirtB01_MSTR_r"(ComponentPose=@n71, bAlphaBoolEnabled=@n72, Translation="0,0,0", Rotation="-10,10,0", Scale="1,1,1") {
    alphaInputType=Bool
    rotationMode=BMM_Additive
    rotationSpace=BCS_BoneSpace
    translationMode=BMM_Additive
    translationSpace=BCS_BoneSpace
    alphaBoolBlend {blendInTime=0.2, blendOption=Linear, blendOutTime=0.2}
    boneToModify {boneName=dyn_SkirtB01_MSTR_r}
    rotation {pitch=-10, yaw=10}
   }
   n74 = "Transform (Modify) Bone - Bone: dyn_SkirtB04_r"!(ComponentPose=@n73, bAlphaBoolEnabled:bool=@n72, Translation="0,0,0", Rotation="0,20,0", Scale="1,1,1") -> (Pose) {
    alphaInputType=Bool
    rotationMode=BMM_Additive
    rotationSpace=BCS_BoneSpace
    translationMode=BMM_Additive
    translationSpace=BCS_BoneSpace
    alphaBoolBlend {blendInTime=0.2, blendOption=Linear, blendOutTime=0.2}
    boneToModify {boneName=dyn_SkirtB04_r}
    rotation {yaw=20}
   }
   n75 = "Transform (Modify) Bone - Bone: dyn_SkirtB05_r"!(ComponentPose=@n74, bAlphaBoolEnabled:bool=@n72, Translation="0,0,0", Rotation="0,20,0", Scale="1,1,1") -> (Pose) {
    alphaInputType=Bool
    rotationMode=BMM_Additive
    rotationSpace=BCS_BoneSpace
    translationMode=BMM_Additive
    translationSpace=BCS_BoneSpace
    alphaBoolBlend {blendInTime=0.2, blendOption=Linear, blendOutTime=0.2}
    boneToModify {boneName=dyn_SkirtB05_r}
    rotation {yaw=20}
   }
   n76 = "Transform (Modify) Bone - Bone: dyn_SkirtB06_r"!(ComponentPose=@n75, bAlphaBoolEnabled:bool=@n72, Translation="0,0,0", Rotation="0,20,0", Scale="1,1,1") -> (Pose) {
    alphaInputType=Bool
    rotationMode=BMM_Additive
    rotationSpace=BCS_BoneSpace
    translationMode=BMM_Additive
    translationSpace=BCS_BoneSpace
    alphaBoolBlend {blendInTime=0.2, blendOption=Linear, blendOutTime=0.2}
    boneToModify {boneName=dyn_SkirtB06_r}
    rotation {yaw=20}
   }
   n77 = "Transform (Modify) Bone - Bone: dyn_SkirtD01_MSTR_r"!(ComponentPose=@n76, bAlphaBoolEnabled:bool=@n72, Translation="0,0,0", Rotation="0,0,0", Scale="1,1,1") -> (Pose) {
    alphaInputType=Bool
    rotationMode=BMM_Additive
    rotationSpace=BCS_BoneSpace
    translationMode=BMM_Additive
    translationSpace=BCS_BoneSpace
    alphaBoolBlend {blendInTime=0.2, blendOption=Linear, blendOutTime=0.2}
    boneToModify {boneName=dyn_SkirtD01_MSTR_r}
   }
   n78 = "Transform (Modify) Bone - Bone: dyn_SkirtD04_r"!(ComponentPose=@n77, bAlphaBoolEnabled:bool=@n72, Translation="0,0,0", Rotation="0,10,0", Scale="1,1,1") -> (Pose) {
    alphaInputType=Bool
    rotationMode=BMM_Additive
    rotationSpace=BCS_BoneSpace
    translationMode=BMM_Additive
    translationSpace=BCS_BoneSpace
    alphaBoolBlend {blendInTime=0.2, blendOption=Linear, blendOutTime=0.2}
    boneToModify {boneName=dyn_SkirtD04_r}
    rotation {yaw=10}
   }
   n79 = "Transform (Modify) Bone - Bone: dyn_SkirtD05_r"!(ComponentPose=@n78, bAlphaBoolEnabled:bool=@n72, Translation="0,0,0", Rotation="0,20,0", Scale="1,1,1") -> (Pose) {
    alphaInputType=Bool
    rotationMode=BMM_Additive
    rotationSpace=BCS_BoneSpace
    translationMode=BMM_Additive
    translationSpace=BCS_BoneSpace
    alphaBoolBlend {blendInTime=0.2, blendOption=Linear, blendOutTime=0.2}
    boneToModify {boneName=dyn_SkirtD05_r}
    rotation {yaw=20}
   }
   n80 = "Transform (Modify) Bone - Bone: dyn_SkirtE01_MSTR_r"!(ComponentPose=@n79, bAlphaBoolEnabled:bool=@n72, Translation="0,0,0", Rotation="0,0,0", Scale="1,1,1") -> (Pose) {
    alphaInputType=Bool
    rotationMode=BMM_Additive
    rotationSpace=BCS_BoneSpace
    translationMode=BMM_Additive
    translationSpace=BCS_BoneSpace
    alphaBoolBlend {blendInTime=0.2, blendOption=Linear, blendOutTime=0.2}
    boneToModify {boneName=dyn_SkirtE01_MSTR_r}
   }
   n81 = "Transform (Modify) Bone - Bone: dyn_SkirtE04_r"(ComponentPose=@n80, bAlphaBoolEnabled=@n72, Translation="0,0,0", Rotation="0,10,0", Scale="1,1,1") {
    alphaInputType=Bool
    rotationMode=BMM_Additive
    rotationSpace=BCS_BoneSpace
    translationMode=BMM_Additive
    translationSpace=BCS_BoneSpace
    alphaBoolBlend {blendInTime=0.2, blendOption=Linear, blendOutTime=0.2}
    boneToModify {boneName=dyn_SkirtE04_r}
    rotation {yaw=10}
   }
   n82 = "Transform (Modify) Bone - Bone: dyn_SkirtE05_r"!(ComponentPose=@n81, bAlphaBoolEnabled:bool=@n72, Translation="0,0,0", Rotation="0,10,0", Scale="1,1,1") -> (Pose) {
    alphaInputType=Bool
    rotationMode=BMM_Additive
    rotationSpace=BCS_BoneSpace
    translationMode=BMM_Additive
    translationSpace=BCS_BoneSpace
    alphaBoolBlend {blendInTime=0.2, blendOption=Linear, blendOutTime=0.2}
    boneToModify {boneName=dyn_SkirtE05_r}
    rotation {yaw=10}
   }
   n83 = "Transform (Modify) Bone - Bone: dyn_SkirtF01_MSTR_c"!(ComponentPose=@n82, bAlphaBoolEnabled:bool=@n72, Translation="0,0,0", Rotation="-10,0,0", Scale="1,1,1") -> (Pose) {
    alphaInputType=Bool
    rotationMode=BMM_Additive
    rotationSpace=BCS_BoneSpace
    translationMode=BMM_Additive
    translationSpace=BCS_BoneSpace
    alphaBoolBlend {blendInTime=0.2, blendOption=Linear, blendOutTime=0.2}
    boneToModify {boneName=dyn_SkirtF01_MSTR_c}
    rotation {pitch=-10}
   }
   n84 = "Transform (Modify) Bone - Bone: dyn_SkirtF04_c"(ComponentPose=@n83, bAlphaBoolEnabled=@n72, Translation="0,0,0", Rotation="-10,0,0", Scale="1,1,1") {
    alphaInputType=Bool
    rotationMode=BMM_Additive
    rotationSpace=BCS_BoneSpace
    translationMode=BMM_Additive
    translationSpace=BCS_BoneSpace
    alphaBoolBlend {blendInTime=0.2, blendOption=Linear, blendOutTime=0.2}
    boneToModify {boneName=dyn_SkirtF04_c}
    rotation {pitch=-10}
   }
   n85 = RigidBody(ComponentPose=@n84, OverrideWorldGravity="0,0,-980") {
    bClampLinearTranslationLimitToRefPose=true
    bOverrideWorldGravity=true
    overridePhysicsAsset=&/M_MED_Noble_Cloak_Crouching_Physics
    simulationSpace=BaseBoneSpace
    alphaBoolBlend {blendOption=Linear}
    baseBoneRef {boneName=root}
    componentAppliedLinearAccClamp=<3000 3000 3000>
    componentLinearAccScale=<0.01 0.5 0.01>
    componentLinearVelScale=<0.01 0.5 0.01>
    overrideWorldGravity {z=-980}
   }
   Save cached pose 'CloackCache'(Pose=Blend Poses by bool(
     BlendPose_0=Component To Local(ComponentPose=@n85)
     BlendPose_1=Blend Poses by bool(BlendPose_0=@n60, BlendPose_1=@n48, bActiveValue=@n61) {blendType=Linear}
     bActiveValue=Get bIsCrouching()
    ) {blendType=Linear}) {cacheName=CloackCache}
   n86 = RigidBody(ComponentPose=Local To Component(LocalPose=Use cached pose 'CloackCache'()), OverrideWorldGravity=Get Gravity Override()) {
    bOverrideWorldGravity=true
    overridePhysicsAsset=&/M_MED_Noble_Hood_Physics
    simulationSpace=BaseBoneSpace
    alphaBoolBlend {blendOption=Linear}
    baseBoneRef {boneName=spine_05}
    componentAppliedLinearAccClamp=<2000 200 200>
    componentLinearAccScale=<0.5 0.05 0.05>
    componentLinearVelScale=<0.1 0.1 0.1>
    overrideWorldGravity {z=-980}
    simSpaceSettings {maxAngularAcceleration=1000, maxAngularVelocity=3, maxLinearAcceleration=250, maxLinearVelocity=500, velocityScaleZ=0, worldAlpha=1}
   }
   n87 = RigidBody(ComponentPose=Local To Component(LocalPose=Use cached pose 'CloackCache'()), OverrideWorldGravity=Get Gravity Override()) {
    bOverrideWorldGravity=true
    overridePhysicsAsset=&/M_MED_Noble_Hood_BackBling_Physics
    simulationSpace=BaseBoneSpace
    alphaBoolBlend {blendOption=Linear}
    baseBoneRef {boneName=spine_05}
    componentAppliedLinearAccClamp=<2000 200 200>
    componentLinearAccScale=<0.5 0.05 0.05>
    componentLinearVelScale=<0.1 0.1 0.1>
    overrideWorldGravity {z=-980}
    simSpaceSettings {maxAngularAcceleration=1000, maxAngularVelocity=3, maxLinearAcceleration=250, maxLinearVelocity=500, velocityScaleZ=0, worldAlpha=1}
   }
   Save cached pose 'RBANCache'(Pose=Blend Poses by bool(
     BlendPose_0=Component To Local(ComponentPose=@n87)
     BlendPose_1=Component To Local(ComponentPose=@n86)
     bActiveValue=Get Is Wearing Backbling()
    ) {blendType=Linear}) {cacheName=RBANCache}
   Get Alt_weapon()
   Get bIsInFrontEnd()
   Get endCrouching() -> (endCrouching:bool)
  ]}
 EventGraph {nodes [
   n1 = Event BlueprintUpdateAnimation()
   > Set Delta Time X(Delta Time X:double=@n1.DeltaTimeX) -> (Output_Get:double)
   > n2 = Sequence!(execute) -> (then_0, then_1, then_2:Exec, then_3:Exec, then_4:Exec, then_5:Exec, then_6:Exec, then_7:Exec, then_8:Exec, then_9:Exec, then_10:Exec)
   > Set Body Mesh(Body Mesh:Skeletal Mesh Component Object Reference=GetPartSkeletalMeshComponent()) -> (Output_Get:Skeletal Mesh Component Object Reference)
   Set Character Instance(
    execute=@n2.then_1
    Character Instance:Fort Player Anim Instance Object Reference=GetOwnerFortPlayerAnimInstance!(self:Custom Character Part Anim Instance Object Reference) -> (ReturnValue:Fort Player Anim Instance Object Reference)
   ) -> (Output_Get:Fort Player Anim Instance Object Reference)
   n3 = GetAnimInstance(self=GetOwnerSkeletalMeshComponent())
   Is Valid(exec=@n2.then_10, InputObject=@n3)
   > Set isCustomEmotePlaying(isCustomEmotePlaying:bool="Equal (Object)"(
     A=GetCurrentActiveMontage!(self:Anim Instance Object Reference=@n3) -> (ReturnValue:Anim Montage Object Reference)
     B=/Game/Animation/Game/MainPlayer/Menu/BR/CustomSkins/Noble/Noble_Male_Idle_Montage
    )) -> (Output_Get:bool)
   Set Thigh Rot R(
    execute=@n2.then_2
    Thigh Rot R:double=GetAlphaFromReferenceTransformDelta(
     SkeletalMeshComponent=Get Body Mesh() -> (Body Mesh:Skeletal Mesh Component Object Reference)
     BoneName=thigh_l
     BaseName=pelvis
     InputRange=(X=-50,Y=0)
     OutputRange=(X=1,Y=0)
     TwistAxis="1,0,0"
    )
   ) -> (Output_Get:double)
   n4 = Gravity Override State Function DEPRECATED(
    execute=@n2.then_3
    CharacterPartAnimInstance=Self-Reference()
    GameTimeInSeconds=Get Delta Time X() -> (Delta Time X:double)
    GravityOverrideParams=Get Gravity Override Params_0() -> (Gravity Override Params_0:Gravity Override Params Struct Structure)
   )
   > Set Gravity Override(Gravity Override=@n4)
   n5 = Cast To FortPlayerAnimInstance(execute=@n2.then_4, Object=GetAnimInstance(self=GetOwnerSkeletalMeshComponent()))
   > Set Alt_weapon(Alt_weapon:bool=false) -> (Output_Get:bool)
   > Set isRifle(isRifle:bool=true) -> (Output_Get:bool)
   > n6 = Switch on EFortWeaponCoreAnimation!(execute, Selection:EFortWeaponCoreAnimation Enum=Get WeaponCoreAnim(self=@n5)) -> (
    Melee:Exec
    Pistol:Exec
    Shotgun:Exec
    PaperBlueprint:Exec
    Rifle:Exec
    MeleeOneHand:Exec
    MachinePistol:Exec
    RocketLauncher:Exec
    GrenadeLauncher:Exec
    GoingCommando:Exec
    AssaultRifle:Exec
    TacticalShotgun:Exec
    SniperRifle:Exec
    TrapPlacement:Exec
    ShoulderLauncher:Exec
    AbilityDecoTool:Exec
    Crossbow:Exec
    C4:Exec
    RemoteControl:Exec
    DualWield:Exec
    AR_BullPup:Exec
    AR_ForwardGrip:Exec
    MedPackPaddles:Exec
    SMG_P90:Exec
    AR_DrumGun:Exec
    Consumable_Small:Exec
    Consumable_Large:Exec
    Balloon:Exec
    MountedTurret:Exec
    CreativeTool:Exec
    ExplosiveBow:Exec
    AshtonIndigo:Exec
    AshtonChicago:Exec
    MeleeDualWield:Exec
    Unarmed:Exec
   )
   Branch(execute=[@n6.AssaultRifle, @n6.Rifle, @n6.Shotgun, @n6.TacticalShotgun])
   > Set Alt_weapon(Alt_weapon=true)
   > Set isRifle(isRifle=false)
   n7 = Branch(execute=[@n6.Melee, @n6.MeleeOneHand])
   > Set isMelee(isMelee:bool=true) -> (Output_Get:bool)
   > n8 = Branch(Condition=Get bIsSprinting())
   > Set isSprintingMelee(isSprintingMelee:bool=true) -> (Output_Get:bool)
   Set isSprintingMelee(execute=@n8.else, isSprintingMelee=false)
   Set isMelee(execute=@n7.else, isMelee=false)
   n9 = PlayerMovement(execute=[@n2.then_5, @n18.then], AnimInstance=Self-Reference(), XDotProductCoefficient=1, YDotProductCoefficient=1, ZDotProductCoefficient=1)
   > n10 = Branch(Condition="float > float"(A=MapRangeClamped(Value=@n9.YMovement, InRangeA=0, InRangeB=1), B=0.5))
   > Set PlayerMovement(PlayerMovement:bool=true) -> (Output_Get:bool)
   Set PlayerMovement(execute=@n10.else, PlayerMovement=false)
   n11 = Get Alt_weapon()
   n12 = Branch(execute=@n2.then_6, Condition=@n11)
   > Set HeadDownADS(HeadDownADS:double=GetAlphaFromReferenceTransformDelta(SkeletalMeshComponent=Get Body Mesh(), BoneName=head, BaseName=neck_01, InputRange=(X=-5,Y=10), TwistAxis="0,0,1")) -> (Output_Get:double)
   Set HeadDownADS(execute=@n12.else, HeadDownADS=0)
   Set Is Wearing Backbling(execute=@n2.then_7, Is Wearing Backbling=CheckForBackbling(self=Self-Reference()))
   n13 = Branch(execute=@n2.then_8, Condition="Not Equal (Float)"(A=@n9.YMovement))
   > Set charMoving(charMoving:bool=false) -> (Output_Get:bool)
   Set charMoving(execute=@n13.else, charMoving=true)
   n14 = Branch(execute=@n2.then_9, Condition=Get bIsTargeting())
   > Set Head Up ADS(Head Up ADS:double=GetAlphaFromReferenceTransformDelta(SkeletalMeshComponent=Get Body Mesh(), BoneName=head, BaseName=neck_01, InputRange=(X=0,Y=-30), TwistAxis="0,0,1")) -> (Output_Get:double)
   Set Head Up ADS(execute=@n14.else, Head Up ADS=0)
   Event OnBeginJumping()
   > Set isJumpingBool(isJumpingBool:bool=true) -> (Output_Get:bool)
   Event OnEndJumping()
   > Set isJumpingBool(isJumpingBool=false)
   n15 = Event OnEndSkydiving()
   ResetDynamics(execute=[@n15.then, @n16.then, @n17.then], InTeleportType=TeleportPhysics)
   > ResetDynamics(InTeleportType=ResetPhysics)
   n16 = Event OnBeginSkydiving()
   Event OnEndCrouching()
   > n17 = Set endCrouching(endCrouching:bool=true) -> (Output_Get:bool)
   Event OnBeginCrouching()
   > Set endCrouching(endCrouching=false)
   n18 = Branch(Condition=Get bIsSprinting())
   Set PlayerMovement(execute=@n18.else, PlayerMovement=false)
   AND Boolean(A=@n11, B=Get bIsTargeting())
  ]}
]
variables [
 Alt_weapon {type=boolean}
 Body Mesh {type=SkeletalMeshComponent}
 Character Instance {type=FortPlayerAnimInstance, default=null}
 charMoving {type=boolean}
 Delta Time X {type=number}
 endCrouching {type=boolean}
 Gravity Override {type=Vector}
 Gravity Override Params {
  type=GravityOverrideParamsStruct
  default {
   skydive_DivingDown {gravityOverride=<-1500 0 1000>}
   skydive_FloatingDown {gravityOverride=<0 -350 100>}
   skydive_GoingUp {gravityOverride=<0 -300 1000>}
   skydive_Parachuting {gravityOverride=<0 -300 200>}
  }
 }
 Gravity Override Params_0 {
  type=GravityOverrideParamsStruct
  default {
   skydive_FloatingDown {windAmplitude=<100 -500 400>, windFrequency=<0.5 7 8>}
   skydive_GoingUp {gravityOverride=<0 -300 600>, windAmplitude=<200 -500 100>, windFrequency=<1.5 7 4>}
  }
 }
 Head Up ADS {type=number}
 HeadDownADS {type=number}
 Is Wearing Backbling {type=boolean}
 isCustomEmotePlaying {type=boolean}
 isJumpingBool {type=boolean}
 isMelee {type=boolean}
 isRifle {type=boolean}
 isSprintingMelee {type=boolean}
 NewVar {type=AnimMontage, default=null}
 NewVar_0 {type=number}
 PlayerMovement {type=boolean}
 Thigh Rot R {type=number}
]
